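# Control Rig for the F_MED_FreeCanyon player body, adding corrective motion to body and costume bones.
# A sequence runs the shared DeformRig function, then poses the right bow bone from an upperarm_r pose reader and constrains bracelet rotation between forearm and hand.
# A DeltaRotation swing/twist function weights a further bracelet modify, whose transform is identity.
# A lowerarm_l pose reader relative to pelvis, with debug drawing enabled, bends the thigh-pouch flower.
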ControlRigBlueprint /BRCosmetics/Characters/Player/Female/Medium/Bodies/F_MED_FreeCanyon/Meshes/F_MED_FreeCanyon_CtrlRig
graphs [
 New Function_ContainedGraph {nodes [
   n1 = RigVMFunctionEntryNode()
   RigVMFunction_MathBoolToInteger = &RigVMFunction_MathBoolToInteger(Value=@n1.onlyTwist)
   RigVMFunction_MathQuaternionSwingTwist = &RigVMFunction_MathQuaternionSwingTwist(
    TwistAxis=@n1.twistAxis
    Input=&RigUnit_GetRelativeTransformForItem(bChildInitial=False, bParentInitial=False, Child=@n1.child, Parent=@n1.parent).RelativeTransform.Rotation
   )
   n2 = RigVMDispatchNode(
    Values=((),())
    Index=@RigVMFunction_MathBoolToInteger.Result
    Values.0=@RigVMFunction_MathQuaternionSwingTwist.Swing
    Values.1=@RigVMFunction_MathQuaternionSwingTwist.Twist
   ) {template="DISPATCH_RigVMDispatch_SelectInt32(in Index,in Values,out Result)"}
   RigVMFunction_MathQuaternionToEuler = &RigVMFunction_MathQuaternionToEuler(Value=@n2.Result, RotationOrder=@n1.Rotation Order)
   > DISPATCH_RigVMDispatch_Print = RigVMDispatchNode(ScreenColor=(R=1,G=1,B=1,A=1), ScreenDuration=1.5, Enabled=@n1.debugRotation, Value=@RigVMFunction_MathQuaternionToEuler.Result) {template="DISPATCH_RigVMDispatch_Print(in Prefix,in Value,in Enabled,in ScreenDuration,in ScreenColor)"}
   Cast = RigVMDispatchNode(Value=@n1.triggerAxis) {template="DISPATCH_RigVMDispatch_CastEnumToInt(in Value,out Result)"}
   DISPATCH_RigVMDispatch_SelectInt32 = RigVMDispatchNode(
    Values=(0,0,0)
    Index=@Cast.Result
    Values.0=@RigVMFunction_MathQuaternionToEuler.Result.X
    Values.1=@RigVMFunction_MathQuaternionToEuler.Result.Y
    Values.2=@RigVMFunction_MathQuaternionToEuler.Result.Z
   ) {template="DISPATCH_RigVMDispatch_SelectInt32(in Index,in Values,out Result)"}
   Remap = &RigVMFunction_MathDoubleRemap(
    bClamp=False
    Value=@DISPATCH_RigVMDispatch_SelectInt32.Result
    TargetMinimum=@n1.outputMinMax.X
    TargetMaximum=@n1.outputMinMax.Y
    SourceMinimum=@n1.rotationMinMax.X
    SourceMaximum=@n1.rotationMinMax.Y
   )
   Absolute = &RigVMFunction_MathDoubleAbs(Value=@Remap.Result)
   Multiply = &RigVMFunction_MathDoubleMul(
    A=&RigVMFunction_MathDoubleClamp(Value=RigVMDispatchNode(True=@Absolute.Result, Condition=@n1.absoluteValue, False=@Remap.Result) {template="DISPATCH_RigVMDispatch_If(in Condition,in True,in False,out Result)"}.Result).Result
    B=@n1.premult
   )
   > RigVMFunctionReturnNode(outputQuat=@n2.Result, outputWeight=@Multiply.Result)
  ]}
 RigVMFunctionLibrary {nodes [DeltaRotation v4 = RigVMCollapseNode(
    ExecuteContext=()
    Rotation Order=ZYX
    child=(Type=Bone,Name="None")
    outputMinMax=(X=0,Y=1)
    parent=(Type=Bone,Name="None")
    premult=1
    rotationMinMax=(X=0,Y=0)
    twistAxis=(X=1,Y=0,Z=0)
   )]}
 RigVMFunction_Sequence_1_ContainedGraph {nodes [
   n1 = RigVMFunctionEntryNode()
   RigVMFunction_Sequence = &RigVMFunction_Sequence(ExecuteContext=@n1.ExecuteContext)
   n2 = &RigVMFunction_Sequence(ExecuteContext=@RigVMFunction_Sequence.B)
   RigVMFunctionReturnNode(A=@RigVMFunction_Sequence.A, B=@n2.A, C=@n2.B)
  ]}
 RigVMModel {nodes [
   RigUnit_BeginExecution = &RigUnit_BeginExecution()
   RigVMFunction_Sequence = RigVMAggregateNode(ExecuteContext=@RigUnit_BeginExecution.ExecutePin)
   FN_DeformRig_FNC = RigVMFunctionReferenceNode(
    DeformRig_ Curve="(deformRig_l_upperarm_up_alpha_72_#1=False,deformRig_r_upperarm_up_alpha_73_#2=False,deformRig_l_upperarm_fr_alpha_74_#3=False,deformRig_r_upperarm_fr_alpha_75_CB51B8 ...(truncated)"
    DeformRig_Bool="(upperarm_1_#4=True,elbow_3_#5=True,wrist_7_#6=True,groin_8_#7=True,glute_17_#8=true,knee_9_40C408324CA261 ...(truncated)"
    DeformRig_Debug=(upperarm_debug_18_#4=False,wrist_debug_20_#6=False,lat_debug_26_#9=False,groinGlute_debug_27_#7=False)
    DeformRig_Value="(upperarm_up_18_#10=(Rotation=(X=0,Y=0,Z=0,W=1),Translation=(X=0,Y=0,Z=8),Scale3D=(X=0.7,Y=1,Z=1)),upperarm_fwd_fr_19_#11=(Rotation=(X=0,Y=0,Z=0,W=1),Translation=(X=0,Y=10,Z=0),Scale3D=(X=0.5,Y=1,Z= ...(truncated)"
    ExecuteContext=@RigVMFunction_Sequence.A
   )
   > n1 = &RigUnit_SphericalPoseReader(
    ActiveRegionScaleFactors=(PositiveWidth=1.5,NegativeWidth=1,PositiveHeight=1,NegativeHeight=1.5)
    ActiveRegionSize=0.1
    Debug=(bDrawDebug=false,bDraw2D=False,bDrawLocalAxes=False,DebugScale=25,DebugSegments=20,DebugThickness=0.25)
    DriverAxis=(X=-1,Y=0,Z=0)
    DriverItem=(Type=Bone,Name="upperarm_r")
    FalloffRegionScaleFactors=(PositiveWidth=1,NegativeWidth=1,PositiveHeight=1,NegativeHeight=1)
    FalloffSize=0.2
    FlipHeightScaling=False
    FlipWidthScaling=False
    OptionalParentItem=(Type=Bone,Name="None")
    RotationOffset=(X=-95.953,Y=9.0367,Z=11.4291)
   )
   > &RigUnit_ModifyTransforms(
    ItemToModify=((Item=(Type=Bone,Name="dyn_bow_r_01"),Transform=(Rotation=(X=0.1341,Y=-0.2336,Z=0.0326,W=0.9625),Translation=(X=0,Y=0,Z=0),Scale3D=(X=1,Y=1,Z=1))))
    WeightMinimum=0
    Weight=@n1.OutputParam
   )
   > &RigUnit_ParentConstraint(
    AdvancedSettings=(InterpolationType=Average,RotationOrderForFilter=XZY)
    Child=(Type=Bone,Name="base_bracelet_r_01")
    Filter=(TranslationFilter=(bX=false,bY=false,bZ=false),RotationFilter=(bX=True,bY=True,bZ=True),ScaleFilter=(bX=True,bY=True,bZ=True))
    Parents=((Item=(Type=Bone,Name="lowerarm_r"),Weight=0.8),(Item=(Type=Bone,Name="hand_r"),Weight=0.2))
    Weight=1
   )
   DeltaRotation v4 = RigVMFunctionReferenceNode(
    Rotation Order=ZYX
    child=(Type=Bone,Name="hand_r")
    outputMinMax=(X=0,Y=1)
    parent=(Type=Control,Name="hand_r_ctrl")
    premult=1
    rotationMinMax=(X=-90,Y=90)
    triggerAxis=X
    twistAxis=(X=1,Y=0,Z=0)
    ExecuteContext=@RigVMFunction_Sequence.B
   )
   > &RigUnit_ModifyTransforms(
    ItemToModify=((Item=(Type=Bone,Name="base_bracelet_r_01"),Transform=(Rotation=(X=0,Y=0,Z=0,W=1),Translation=(X=0,Y=0,Z=0),Scale3D=(X=1,Y=1,Z=1))))
    WeightMinimum=0
    Weight=@"DeltaRotation v4".outputWeight
   )
   n2 = &RigUnit_SphericalPoseReader(
    ActiveRegionScaleFactors=(PositiveWidth=2,NegativeWidth=1,PositiveHeight=1.5,NegativeHeight=1.5)
    ActiveRegionSize=0.05
    Debug=(bDrawDebug=True,bDraw2D=False,bDrawLocalAxes=False,DebugScale=40,DebugSegments=20,DebugThickness=0.25)
    DriverAxis=(X=1,Y=0,Z=0)
    DriverItem=(Type=Bone,Name="lowerarm_l")
    FalloffRegionScaleFactors=(PositiveWidth=3,NegativeWidth=0.5,PositiveHeight=4,NegativeHeight=1)
    FalloffSize=0.1
    FlipHeightScaling=False
    FlipWidthScaling=False
    OptionalParentItem=(Type=Bone,Name="pelvis")
    RotationOffset=(X=123.6997,Y=-32.2272,Z=-61.7447)
    ExecutePin=@RigVMFunction_Sequence.C
   )
   > &RigUnit_ModifyTransforms(
    ItemToModify="((Item=(Type=Bone,Name=\"dyn_flower_02\"),Transform=(Rotation=(X=0,Y=0,Z=0.1736,W=0.9848),Translation=(X=0,Y=0,Z=0),Scale3D=(X=1,Y=1,Z=1))),(Item=(Type=Bone,Name=\"dyn_flower_leaf_r_01\"),Transform=(Rotation=(X=0,Y=0.2588,Z=0,W=0.9659),Translation=(X=0,Y=0,Z=0 ...(truncated)"
    WeightMinimum=0
    Weight=@n2.OutputParam
   )
  ]}
]
hierarchy [
 root {loc=<0 0 0>, rot=<0 0 0 1>}
 pelvis {parent=root, loc=<0 -0.5301 98.0571>, rot=<0.0127 0.707 -0.0127 -0.707>}
 spine_01 {parent=pelvis, loc=<6.1208 -0.9328 0>, rot=<0 0 0.034 -0.9994>}
 spine_02 {parent=spine_01, loc=<6.5997 -0.2095 0>, rot=<0 0 0.0238 -0.9997>}
 spine_03 {parent=spine_02, loc=<6.4914 -0.1237 0>, rot=<0 0 0.0149 -0.9999>}
 spine_04 {parent=spine_03, loc=<6.9598 0.0794 0>, rot=<0 0 -0.0132 -0.9999>}
 spine_05 {parent=spine_04, loc=<6.9987 0 0>, rot=<0 0 -0.0015 -1>}
 neck_01 {parent=spine_05, loc=<11.3138 0.1529 0>, rot=<0 0 -0.2003 -0.9797>}
 neck_02 {parent=neck_01, loc=<4.5627 0 0>, rot=<0 0 0.0504 -0.9987>}
 head {parent=neck_02, loc=<5.3914 0 0>, rot=<0 0 0.115 -0.9934>}
 clavicle_l {parent=spine_05, loc=<5.4831 -0.1053 -3.2946>, rot=<0.0465 -0.7289 -0.009 -0.6829>}
 upperarm_l {parent=clavicle_l, loc=<9.8658 -1.2136 -0.4105>, rot=<0.0164 -0.3245 -0.0269 -0.9454>}
 lowerarm_l {parent=upperarm_l, loc=<26.465 -0.8345 0.259>, rot=<0.0036 0.0161 -0.2298 -0.9731>}
 hand_l {parent=lowerarm_l, loc=<22.7355 -0.3203 -0.6105>, rot=<-0.6784 -0.0223 0.0735 -0.7307>}
 thumb_01_l {parent=hand_l, loc=<1.3118 -0.3191 -2.459>, rot=<0.6546 -0.4042 0.0664 -0.6354>}
 thumb_02_l {parent=thumb_01_l, loc=<5.9336 -0.0448 0.0401>, rot=<-0.0037 0.0213 0.298 -0.9543>}
 thumb_03_l {parent=thumb_02_l, loc=<2.2223 0.3682 -0.0907>, rot=<0.0118 -0.013 0.0066 -0.9998>}
 index_metacarpal_l {parent=hand_l, loc=<1.6535 0.0183 -2.96>, rot=<0.0158 -0.0124 -0.0173 -0.9996>}
 index_01_l {parent=index_metacarpal_l, loc=<8.3008 0.0357 -0.0524>, rot=<-0.0044 0.0485 0.0924 -0.9945>}
 index_02_l {parent=index_01_l, loc=<3.3587 -0.3004 0.0452>, rot=<-0.002 0.0027 0.3225 -0.9466>}
 index_03_l {parent=index_02_l, loc=<2.5578 -0.123 -0.0252>, rot=<-0.002 0.0067 0.2558 -0.9667>}
 middle_metacarpal_l {parent=hand_l, loc=<2.1856 -0.1725 -0.5984>, rot=<0.006 0.0168 -0.0329 -0.9993>}
 middle_01_l {parent=middle_metacarpal_l, loc=<7.3275 -0.2447 0.0129>, rot=<-0.0208 0.0042 0.0858 -0.9961>}
 middle_02_l {parent=middle_01_l, loc=<4.0123 0.1401 0.1175>, rot=<-0.0218 0.0258 0.3512 -0.9357>}
 middle_03_l {parent=middle_02_l, loc=<2.6077 0.0534 0.1373>, rot=<0.0082 0.0188 0.1708 -0.9851>}
 ring_metacarpal_l {parent=hand_l, loc=<2.2074 -0.3625 1.3638>, rot=<-0.0284 0.0722 -0.0458 -0.9959>}
 ring_01_l {parent=ring_metacarpal_l, loc=<7.0295 -0.6197 -0.1446>, rot=<-0.0489 -0.0252 0.1892 -0.9804>}
 ring_02_l {parent=ring_01_l, loc=<3.6174 0.4647 -0.1027>, rot=<0 0 0.2806 -0.9598>}
 ring_03_l {parent=ring_02_l, loc=<2.4152 0.114 0.0199>, rot=<0 0 0.1189 -0.9929>}
 pinky_metacarpal_l {parent=hand_l, loc=<1.7616 -0.9753 3.3814>, rot=<-0.0397 0.0814 -0.069 -0.9935>}
 pinky_01_l {parent=pinky_metacarpal_l, loc=<7.1709 -0.4831 -0.1094>, rot=<-0.0654 -0.0035 0.2242 -0.9723>}
 pinky_02_l {parent=pinky_01_l, loc=<2.8274 0.0893 -0.1043>, rot=<0.0037 0.0011 0.2854 -0.9584>}
 pinky_03_l {parent=pinky_02_l, loc=<2.2046 0.0307 -0.0078>, rot=<-0.0002 0.0098 0.1445 -0.9895>}
 hand_attach_l {parent=hand_l, loc=<8.5 -3 0>, rot=<0.7071 0.7071 0 0>}
 weapon_l {parent=hand_l, loc=<-0.1974 -3.4111 0.3811>, rot=<0 0 0 1>}
 deform_wrist_dn_base_l {parent=hand_l, loc=<0 -1.5326 0>, rot=<0 0 0 -1>}
 deform_wrist_dn_l {parent=deform_wrist_dn_base_l, loc=<0 0 0>, rot=<0 0 0 1>}
 deform_wrist_up_base_l {parent=hand_l, loc=<0 2.3009 0>, rot=<0 0 0 -1>}
 deform_wrist_up_l {parent=deform_wrist_up_base_l, loc=<0 0 0>, rot=<0 0 0 1>}
 lowerarm_twist_02_l {parent=lowerarm_l, loc=<10.6707 0 0>, rot=<0 0 0 1>}
 lowerarm_twist_01_l {parent=lowerarm_l, loc=<18.316 0 0>, rot=<0 0 0 1>}
 deform_elbow_in_base_l {parent=lowerarm_l, loc=<0.1834 3.6062 -0.2572>, rot=<0 0 0 -1>}
 deform_elbow_in_l {parent=deform_elbow_in_base_l, loc=<0 0 0>, rot=<0 0 0 1>}
 deform_elbow_out_base_l {parent=lowerarm_l, loc=<0 -4 0>, rot=<0 0 0 1>}
 deform_elbow_out_l {parent=deform_elbow_out_base_l, loc=<0 0 0>, rot=<0 0 0 1>}
 base_slevee_main_l {parent=lowerarm_l, loc=<0 0 0>, rot=<0 0 0 1>}
 dyn_slevee_main_l {parent=base_slevee_main_l, loc=<0 0 0>, rot=<0 0 0 1>}
 base_slevee_dwn_l {parent=dyn_slevee_main_l, loc=<4.8077 -0.2596 -4.4624>, rot=<0 -0.1113 0 -0.9938>}
 dyn_slevee_dwn_l {parent=base_slevee_dwn_l, loc=<0 0 0>, rot=<0 0 0 1>}
 base_slevee_fringes_01_l {parent=dyn_slevee_dwn_l, loc=<-0.0491 -1.926 -1.4195>, rot=<0.0947 -0.3639 0.1158 -0.9194>}
 dyn_slevee_fringes_01_l {parent=base_slevee_fringes_01_l, loc=<0 0 0>, rot=<0 0 0 1>}
 dyn_slevee_fringes_01a_l {parent=dyn_slevee_fringes_01_l, loc=<7.4266 -0.2375 0>, rot=<0 0 0 1>}
 base_slevee_fringes_02_l {parent=dyn_slevee_dwn_l, loc=<4.415 -2.728 -1.7736>, rot=<0.0947 -0.3639 0.1158 -0.9194>}
 dyn_slevee_fringes_02_l {parent=base_slevee_fringes_02_l, loc=<0 0 0>, rot=<0 0 0 1>}
 dyn_slevee_fringes_02a_l {parent=dyn_slevee_fringes_02_l, loc=<7.0896 -0.2438 0>, rot=<0 0 0 1>}
 base_slevee_fringes_03_l {parent=dyn_slevee_dwn_l, loc=<9.1364 -3.6444 -2.4225>, rot=<0.0947 -0.3639 0.1158 -0.9194>}
 dyn_slevee_fringes_03_l {parent=base_slevee_fringes_03_l, loc=<0 0 0>, rot=<0 0 0 1>}
 dyn_slevee_fringes_03a_l {parent=dyn_slevee_fringes_03_l, loc=<6.4873 -0.1477 0>, rot=<0 0 0 1>}
 base_slevee_fringes_04_l {parent=dyn_slevee_dwn_l, loc=<13.8543 -4.7942 -3.134>, rot=<0.0947 -0.3639 0.1158 -0.9194>}
 dyn_slevee_fringes_04_l {parent=base_slevee_fringes_04_l, loc=<0 0 0>, rot=<0 0 0 1>}
 dyn_slevee_fringes_04a_l {parent=dyn_slevee_fringes_04_l, loc=<5.9183 -0.1477 0>, rot=<0 0 0 1>}
 base_slevee_uppr_l {parent=dyn_slevee_main_l, loc=<4.6545 0.2381 3.4121>, rot=<0 0 0 1>}
 dyn_slevee_uppr_l {parent=base_slevee_uppr_l, loc=<0 0 0>, rot=<0 0 0 1>}
 upperarm_twist_01_l {parent=upperarm_l, loc=<7.2135 -0.1134 0>, rot=<0 0 0 1>}
 deform_upperarm_bk_base_l {parent=upperarm_twist_01_l, loc=<-6.4053 -4.3746 0.314>, rot=<0 0 0 -1>}
 deform_upperarm_bk_l {parent=deform_upperarm_bk_base_l, loc=<0 0 0>, rot=<0 0 0 1>}
 deform_upperarm_fr_base_l {parent=upperarm_twist_01_l, loc=<-5.6312 5.092 0.3354>, rot=<0 0 0 -1>}
 deform_upperarm_fr_l {parent=deform_upperarm_fr_base_l, loc=<0 0 0>, rot=<0 0 0 1>}
 deform_upperarm_up_base_l {parent=upperarm_twist_01_l, loc=<-8.891 0.4825 3.9517>, rot=<0 0 0 -1>}
 deform_upperarm_up_l {parent=deform_upperarm_up_base_l, loc=<0 0 0>, rot=<0 0 0 1>}
 upperarm_twist_02_l {parent=upperarm_l, loc=<19.7998 -0.1134 0>, rot=<0 0 0 1>}
 clavicle_r {parent=spine_05, loc=<5.4831 -0.1053 3.2946>, rot=<0.7289 0.0465 -0.6829 0.009>}
 upperarm_r {parent=clavicle_r, loc=<-9.8658 1.2136 0.4105>, rot=<0.0164 -0.3245 -0.0269 -0.9454>}
 lowerarm_r {parent=upperarm_r, loc=<-26.465 0.8345 -0.259>, rot=<0.0036 0.0161 -0.2298 -0.9731>}
 hand_r {parent=lowerarm_r, loc=<-22.7355 0.3203 0.6106>, rot=<-0.6784 -0.0223 0.0735 -0.7307>}
 thumb_01_r {parent=hand_r, loc=<-1.3118 0.3191 2.459>, rot=<0.6546 -0.4042 0.0664 -0.6354>}
 thumb_02_r {parent=thumb_01_r, loc=<-5.9336 0.0448 -0.0401>, rot=<-0.0037 0.0213 0.298 -0.9543>}
 thumb_03_r {parent=thumb_02_r, loc=<-2.2223 -0.3682 0.0907>, rot=<0.0118 -0.013 0.0066 -0.9998>}
 index_metacarpal_r {parent=hand_r, loc=<-1.6535 -0.0183 2.96>, rot=<0.0158 -0.0124 -0.0173 -0.9996>}
 index_01_r {parent=index_metacarpal_r, loc=<-8.3008 -0.0357 0.0524>, rot=<-0.0044 0.0485 0.0924 -0.9945>}
 index_02_r {parent=index_01_r, loc=<-3.3587 0.3004 -0.0452>, rot=<-0.002 0.0027 0.3225 -0.9466>}
 index_03_r {parent=index_02_r, loc=<-2.5578 0.123 0.0252>, rot=<-0.002 0.0067 0.2558 -0.9667>}
 middle_metacarpal_r {parent=hand_r, loc=<-2.1856 0.1725 0.5984>, rot=<0.006 0.0168 -0.0329 -0.9993>}
 middle_01_r {parent=middle_metacarpal_r, loc=<-7.3275 0.2447 -0.0129>, rot=<-0.0208 0.0042 0.0858 -0.9961>}
 middle_02_r {parent=middle_01_r, loc=<-4.0123 -0.1401 -0.1175>, rot=<-0.0218 0.0258 0.3512 -0.9357>}
 middle_03_r {parent=middle_02_r, loc=<-2.6077 -0.0534 -0.1373>, rot=<0.0082 0.0188 0.1708 -0.9851>}
 ring_metacarpal_r {parent=hand_r, loc=<-2.2074 0.3625 -1.3638>, rot=<-0.0284 0.0722 -0.0458 -0.9959>}
 ring_01_r {parent=ring_metacarpal_r, loc=<-7.0295 0.6197 0.1446>, rot=<-0.0489 -0.0252 0.1892 -0.9804>}
 ring_02_r {parent=ring_01_r, loc=<-3.6174 -0.4647 0.1027>, rot=<0 0 0.2806 -0.9598>}
 ring_03_r {parent=ring_02_r, loc=<-2.4152 -0.114 -0.0199>, rot=<0 0 0.1189 -0.9929>}
 pinky_metacarpal_r {parent=hand_r, loc=<-1.7616 0.9753 -3.3814>, rot=<-0.0397 0.0814 -0.069 -0.9935>}
 pinky_01_r {parent=pinky_metacarpal_r, loc=<-7.1709 0.4831 0.1094>, rot=<-0.0654 -0.0035 0.2242 -0.9723>}
 pinky_02_r {parent=pinky_01_r, loc=<-2.8274 -0.0893 0.1043>, rot=<0.0037 0.0011 0.2854 -0.9584>}
 pinky_03_r {parent=pinky_02_r, loc=<-2.2046 -0.0307 0.0078>, rot=<-0.0002 0.0098 0.1445 -0.9895>}
 weapon_r {parent=hand_r, loc=<0.1974 3.4111 -0.3811>, rot=<0 0 0 1>}
 hand_attach_r {parent=hand_r, loc=<-8.5001 2.9999 0>, rot=<0 0 -0.7071 -0.7071>}
 deform_wrist_dn_base_r {parent=hand_r, loc=<-0.0001 1.5328 0>, rot=<0 0 0 -1>}
 deform_wrist_dn_r {parent=deform_wrist_dn_base_r, loc=<0 0 0>, rot=<0 0 0 1>}
 deform_wrist_up_base_r {parent=hand_r, loc=<-0.0002 -2.3007 -0.0001>, rot=<0 0 0 -1>}
 deform_wrist_up_r {parent=deform_wrist_up_base_r, loc=<0 0 0>, rot=<0 0 0 1>}
 lowerarm_twist_02_r {parent=lowerarm_r, loc=<-10.6707 0 0>, rot=<0 0 0 -1>}
 lowerarm_twist_01_r {parent=lowerarm_r, loc=<-18.316 0 0>, rot=<0 0 0 -1>}
 deform_elbow_in_base_r {parent=lowerarm_r, loc=<-0.1834 -3.6062 0.2572>, rot=<0 0 0 -1>}
 deform_elbow_in_r {parent=deform_elbow_in_base_r, loc=<0 0 0>, rot=<0 0 0 1>}
 deform_elbow_out_base_r {parent=lowerarm_r, loc=<0 4 -0.0001>, rot=<0 0 0 -1>}
 deform_elbow_out_r {parent=deform_elbow_out_base_r, loc=<0 0 0>, rot=<0 0 0 1>}
 base_bracelet_r_01 {parent=lowerarm_r, loc=<-22.4543 0.2885 1.0438>, rot=<0 -1 0 0>}
 dyn_bracelet_r_01 {parent=base_bracelet_r_01, loc=<0 0 0>, rot=<0 0 0 1>}
 base_slevee_main_r {parent=lowerarm_r, loc=<0 0 0>, rot=<0 0 0 1>}
 dyn_slevee_main_r {parent=base_slevee_main_r, loc=<0 0 0>, rot=<0 0 0 1>}
 base_slevee_uppr_r {parent=dyn_slevee_main_r, loc=<-4.6546 -0.238 -3.412>, rot=<0 -1 0 0>}
 dyn_slevee_uppr_r {parent=base_slevee_uppr_r, loc=<0 0 0>, rot=<0 0 0 1>}
 base_slevee_dwn_r {parent=dyn_slevee_main_r, loc=<-4.808 0.26 4.462>, rot=<0 -0.9938 0 0.1113>}
 dyn_slevee_dwn_r {parent=base_slevee_dwn_r, loc=<0 0 0>, rot=<0 0 0 -1>}
 base_slevee_fringes_01_r {parent=dyn_slevee_dwn_r, loc=<-0.0487 1.9262 -1.4198>, rot=<-0.0947 -0.3639 -0.1158 -0.9194>}
 dyn_slevee_fringes_01_r {parent=base_slevee_fringes_01_r, loc=<0 0 0>, rot=<0 0 0 1>}
 dyn_slevee_fringes_01a_r {parent=dyn_slevee_fringes_01_r, loc=<7.4259 0.2376 0>, rot=<0 0 0 1>}
 base_slevee_fringes_02_r {parent=dyn_slevee_dwn_r, loc=<4.4148 2.7279 -1.7734>, rot=<-0.0947 -0.3639 -0.1158 -0.9194>}
 dyn_slevee_fringes_02_r {parent=base_slevee_fringes_02_r, loc=<0 0 0>, rot=<0 0 0 1>}
 dyn_slevee_fringes_02a_r {parent=dyn_slevee_fringes_02_r, loc=<7.09 0.2438 0>, rot=<0 0 0 1>}
 base_slevee_fringes_03_r {parent=dyn_slevee_dwn_r, loc=<9.1368 3.6445 -2.4227>, rot=<-0.0947 -0.3639 -0.1158 -0.9194>}
 dyn_slevee_fringes_03_r {parent=base_slevee_fringes_03_r, loc=<0 0 0>, rot=<0 0 0 1>}
 dyn_slevee_fringes_03a_r {parent=dyn_slevee_fringes_03_r, loc=<6.4869 0.1477 0>, rot=<0 0 0 1>}
 base_slevee_fringes_04_r {parent=dyn_slevee_dwn_r, loc=<13.8543 4.7942 -3.134>, rot=<-0.0947 -0.3639 -0.1158 -0.9194>}
 dyn_slevee_fringes_04_r {parent=base_slevee_fringes_04_r, loc=<0 0 0>, rot=<0 0 0 1>}
 dyn_slevee_fringes_04a_r {parent=dyn_slevee_fringes_04_r, loc=<5.9184 0.1477 0>, rot=<0 0 0 1>}
 upperarm_twist_01_r {parent=upperarm_r, loc=<-7.2135 0.1134 0>, rot=<0 0 0 -1>}
 deform_upperarm_bk_base_r {parent=upperarm_twist_01_r, loc=<6.4053 4.3746 -0.314>, rot=<0 0 0 -1>}
 deform_upperarm_bk_r {parent=deform_upperarm_bk_base_r, loc=<0 0 0>, rot=<0 0 0 1>}
 deform_upperarm_fr_base_r {parent=upperarm_twist_01_r, loc=<5.6312 -5.092 -0.3354>, rot=<0 0 0 -1>}
 deform_upperarm_fr_r {parent=deform_upperarm_fr_base_r, loc=<0 0 0>, rot=<0 0 0 1>}
 deform_upperarm_up_base_r {parent=upperarm_twist_01_r, loc=<8.891 -0.4825 -3.9517>, rot=<0 0 0 -1>}
 deform_upperarm_up_r {parent=deform_upperarm_up_base_r, loc=<0 0 0>, rot=<0 0 0 1>}
 upperarm_twist_02_r {parent=upperarm_r, loc=<-19.7998 0.1134 0>, rot=<0 0 0 -1>}
 attach_backpack {parent=spine_05, loc=<-0.123 -13.473 0>, rot=<-0.0093 -0.707 0.0093 -0.707>}
 Backpack_BR {parent=attach_backpack, type=Null, loc=<0 0 0>, rot=<0 0 0 1>}
 attach_cape {parent=spine_05, loc=<9.017 -11.714 0>, rot=<0.0616 -0.7044 -0.0616 -0.7044>}
 cape {parent=attach_cape, type=Null, loc=<0 0 0>, rot=<0 0 0 1>}
 base_bow_01 {parent=spine_05, loc=<11.1723 5.6894 -0.3054>, rot=<0.402 0.5817 0.5817 0.402>}
 dyn_bow_01 {parent=base_bow_01, loc=<0 0 0>, rot=<0 0 0 1>}
 dyn_bow_l_01 {parent=dyn_bow_01, loc=<0.1676 -2.2175 -1.3402>, rot=<0.3468 -0.4666 0.4081 -0.7039>}
 dyn_bow_r_01 {parent=dyn_bow_01, loc=<1.3974 1.8729 -1.2924>, rot=<-0.5604 -0.306 -0.0899 -0.7644>}
 deform_pec_aim_base_l {parent=spine_04, loc=<3.2456 9.3309 -1.2237>, rot=<0.0603 -0.4033 0.3318 -0.8507>}
 deform_pec_aim_l {parent=deform_pec_aim_base_l, loc=<0 0 0>, rot=<0 0 0 1>}
 deform_pec_base_l {parent=deform_pec_aim_l, loc=<9.3665 0 0>, rot=<0 0 0 1>}
 deform_pec_l {parent=deform_pec_base_l, loc=<0 0 0>, rot=<0 0 0 1>}
 deform_pec_aim_base_r {parent=spine_04, loc=<3.2452 9.3309 1.2237>, rot=<-0.4033 -0.0603 0.8507 0.3318>}
 deform_pec_aim_r {parent=deform_pec_aim_base_r, loc=<0 0 0>, rot=<0 0 0 1>}
 deform_pec_base_r {parent=deform_pec_aim_r, loc=<-9.3667 -0.0003 -0.0004>, rot=<0 0 0 1>}
 deform_pec_r {parent=deform_pec_base_r, loc=<0 0 0>, rot=<0 0 0 1>}
 base_blouse_flap_01_l {parent=spine_03, loc=<-0.3074 10.1224 -5.9956>, rot=<0.0733 -0.9603 0.2534 -0.0908>}
 dyn_blouse_flap_01_l {parent=base_blouse_flap_01_l, loc=<0 0 0>, rot=<0 0 0 1>}
 dyn_blouse_flap_02_l {parent=dyn_blouse_flap_01_l, loc=<4.7049 0 0>, rot=<0 0 0 1>}
 base_blouse_flap_01_r {parent=spine_03, loc=<-0.3071 10.1224 5.9956>, rot=<0.9603 0.0733 -0.0908 -0.2534>}
 dyn_blouse_flap_01_r {parent=base_blouse_flap_01_r, loc=<0 0 0>, rot=<0 -1 0 0>}
 dyn_blouse_flap_02_r {parent=dyn_blouse_flap_01_r, loc=<4.7055 0 -0.0001>, rot=<0 0 0 1>}
 thigh_l {parent=pelvis, loc=<-1.0855 -0.2109 -8.7053>, rot=<0.0009 0.0355 0.0155 -0.9993>}
 calf_l {parent=thigh_l, loc=<-42.5284 -0.0019 -0.7297>, rot=<0.0006 -0.0161 -0.0391 -0.9991>}
 foot_l {parent=calf_l, loc=<-46.7401 -2.5014 -0.592>, rot=<0.0427 -0.0238 0.0353 -0.9982>}
 ball_l {parent=foot_l, loc=<-3.5658 10.8066 -0.1544>, rot=<0.0001 -0.0001 -0.7063 -0.7079>}
 calf_twist_02_l {parent=calf_l, loc=<-19.1546 -1.7943 -0.2638>, rot=<0.0237 0.0095 -0.0035 -0.9997>}
 calf_twist_01_l {parent=calf_l, loc=<-35.8483 -1.9197 -0.5794>, rot=<0.0237 0.0095 -0.0035 -0.9997>}
 deform_knee_in_base_l {parent=calf_l, loc=<0.8031 -5.0231 0.2456>, rot=<0 0 0 -1>}
 deform_knee_in_l {parent=deform_knee_in_base_l, loc=<0 0 0>, rot=<0 0 0 1>}
 deform_knee_out_base_l {parent=calf_l, loc=<0 4 0>, rot=<0 0 0 -1>}
 deform_knee_out_l {parent=deform_knee_out_base_l, loc=<0 0 0>, rot=<0 0 0 1>}
 base_calf_pant_01_l {parent=calf_l, loc=<-20.0986 -2.7789 -0.8613>, rot=<0.0035 -0.9997 0.0237 -0.0095>}
 dyn_calf_pant_01_l {parent=base_calf_pant_01_l, loc=<0 0 0>, rot=<0 0 0 1>}
 thigh_twist_01_l {parent=thigh_l, loc=<-27.952 0 0>, rot=<0.0013 0 0 -1>}
 deform_groin_base_l {parent=thigh_l, loc=<0.4801 5.6776 -0.2486>, rot=<0 0 0 -1>}
 deform_groin_l {parent=deform_groin_base_l, loc=<0 0 0>, rot=<0 0 0 1>}
 base_pouch {parent=thigh_l, loc=<-19.4121 -1.3954 -7.8785>, rot=<0.0353 -0.0438 0.9984 0.004>}
 dyn_pouch {parent=base_pouch, loc=<0 0 0>, rot=<0 0 0 1>}
 base_flower_01 {parent=dyn_pouch, loc=<-1.5298 3.6682 -1.1474>, rot=<0.097 -0.8742 0.4475 -0.1615>}
 dyn_flower_01 {parent=base_flower_01, loc=<0 0 0>, rot=<0 0 0 1>}
 dyn_flower_02 {parent=dyn_flower_01, loc=<6.5269 0.1291 -0.2965>, rot=<0.0004 0.0776 -0.0046 -0.997>}
 dyn_flower_03 {parent=dyn_flower_02, loc=<3.8282 0.0206 0.0093>, rot=<0 0.0773 0 -0.997>}
 dyn_flower_leaf_l_01 {parent=dyn_flower_03, loc=<4.7419 2.1336 0.5048>, rot=<-0.5912 -0.3879 -0.4707 -0.5277>}
 dyn_flower_leaf_l_02 {parent=dyn_flower_leaf_l_01, loc=<1.9829 0 0>, rot=<0 -0.169 0 -0.9856>}
 dyn_flower_leaf_l_03 {parent=dyn_flower_leaf_l_02, loc=<2.4156 0.7377 -1.3081>, rot=<0 -0.4659 0 -0.8848>}
 dyn_flower_leaf_bck_01 {parent=dyn_flower_03, loc=<5.0639 -0.515 -1.0272>, rot=<0.2409 -0.4534 0.0824 -0.8542>}
 dyn_flower_leaf_bck_02 {parent=dyn_flower_leaf_bck_01, loc=<1.7759 0.0149 0.1558>, rot=<-0.035 -0.1676 0.0502 -0.984>}
 dyn_flower_leaf_bck_03 {parent=dyn_flower_leaf_bck_02, loc=<1.8568 -0.0296 -1.8214>, rot=<0 -0.4659 0 -0.8848>}
 dyn_flower_leaf_r_01 {parent=dyn_flower_03, loc=<5.0283 -2.3718 1.3116>, rot=<0.601 -0.2364 0.4936 -0.5824>}
 dyn_flower_leaf_r_02 {parent=dyn_flower_leaf_r_01, loc=<2.3909 0.1177 0.0081>, rot=<0.0209 -0.164 0.0044 -0.9862>}
 dyn_flower_leaf_r_03 {parent=dyn_flower_leaf_r_02, loc=<3.1949 0.0611 -1.1195>, rot=<0 -0.2437 0 -0.9699>}
 dyn_flower_leaf_frnt_01 {parent=dyn_flower_03, loc=<4.5518 -0.0727 3.3974>, rot=<0.8118 -0.0829 0.5777 0.017>}
 dyn_flower_leaf_frnt_02 {parent=dyn_flower_leaf_frnt_01, loc=<2.1101 0.0224 -0.2043>, rot=<0.0206 -0.1711 -0.0323 -0.9845>}
 dyn_flower_leaf_frnt_03 {parent=dyn_flower_leaf_frnt_02, loc=<2.6803 0.3556 -2.1902>, rot=<-0.1223 -0.4227 -0.084 -0.894>}
 dyn_flower_bck_01 {parent=dyn_flower_03, loc=<4.5195 0.0176 0.379>, rot=<0.9691 -0.0364 -0.0628 -0.2356>}
 dyn_flower_bck_02 {parent=dyn_flower_bck_01, loc=<1.2019 -0.0131 0.2379>, rot=<0 0.0938 0 -0.9956>}
 dyn_flower_mid_01 {parent=dyn_flower_03, loc=<4.5759 0.311 0.7693>, rot=<0.9689 -0.155 -0.0117 -0.1926>}
 dyn_flower_mid_02 {parent=dyn_flower_mid_01, loc=<1.3967 -0.1298 -0.03>, rot=<0.0881 -0.0177 0.0107 -0.9959>}
 dyn_flower_frnt_01 {parent=dyn_flower_03, loc=<4.2012 -0.1415 1.0172>, rot=<0.9807 0.1067 0.0433 -0.1578>}
 dyn_flower_frnt_02 {parent=dyn_flower_frnt_01, loc=<2.1406 -0.2017 -0.2313>, rot=<0.0881 -0.0177 0.0107 -0.9959>}
 dyn_leaf_01 {parent=dyn_flower_01, loc=<9.9063 1.9581 0.7068>, rot=<0.8467 0.1903 0.3494 0.3532>}
 dyn_leaf_02 {parent=dyn_flower_01, loc=<8.7244 2.8558 -2.621>, rot=<-0.673 -0.4208 -0.1704 -0.5839>}
 dyn_leaf_03 {parent=dyn_flower_01, loc=<8.1815 4.2482 -4.4319>, rot=<0.8666 0.2328 0.1269 0.4228>}
 dyn_leaf_04 {parent=dyn_flower_01, loc=<4.9279 4.5722 -4.9773>, rot=<-0.6836 -0.4652 -0.12 -0.5494>}
 dyn_leaf_05 {parent=dyn_flower_01, loc=<6.0489 5.5094 -7.7206>, rot=<-0.5246 -0.4376 -0.0206 -0.73>}
 dyn_leaf_06 {parent=dyn_flower_01, loc=<5.144 3.587 -8.4177>, rot=<-0.0304 -0.5581 0.0105 -0.8291>}
 dyn_leaf_07 {parent=dyn_flower_01, loc=<3.4802 1.8569 -6.3674>, rot=<0.1304 -0.6382 -0.0051 -0.7587>}
 dyn_scissor_01 {parent=dyn_pouch, loc=<-0.4996 -8.1964 -2.2014>, rot=<-0.7986 -0.0385 -0.0514 -0.5984>}
 thigh_r {parent=pelvis, loc=<-1.0855 -0.2109 8.7053>, rot=<0.0355 -0.0009 0.9993 0.0155>}
 calf_r {parent=thigh_r, loc=<42.5284 0.0019 0.7297>, rot=<0.0006 -0.0161 -0.0391 -0.9991>}
 foot_r {parent=calf_r, loc=<46.7401 2.5014 0.592>, rot=<0.0427 -0.0238 0.0353 -0.9982>}
 ball_r {parent=foot_r, loc=<3.5658 -10.8066 0.1544>, rot=<0.0001 -0.0001 -0.7063 -0.7079>}
 calf_twist_02_r {parent=calf_r, loc=<19.1546 1.7943 0.2638>, rot=<0.0237 0.0095 -0.0035 -0.9997>}
 calf_twist_01_r {parent=calf_r, loc=<35.8483 1.9197 0.5794>, rot=<0.0237 0.0095 -0.0035 -0.9997>}
 deform_knee_in_base_r {parent=calf_r, loc=<-0.8031 5.0231 -0.2457>, rot=<0 0 0 -1>}
 deform_knee_in_r {parent=deform_knee_in_base_r, loc=<0 0 0>, rot=<0 0 0 1>}
 deform_knee_out_base_r {parent=calf_r, loc=<0 -4 0>, rot=<0 0 0 -1>}
 deform_knee_out_r {parent=deform_knee_out_base_r, loc=<0 0 0>, rot=<0 0 0 1>}
 base_calf_pant_01_r {parent=calf_r, loc=<20.0986 2.7789 0.8613>, rot=<0.0237 0.0095 -0.0035 -0.9997>}
 dyn_calf_pant_01_r {parent=base_calf_pant_01_r, loc=<0 0 0>, rot=<0 0 0 1>}
 thigh_twist_01_r {parent=thigh_r, loc=<27.952 0 0>, rot=<0.0013 0 0 -1>}
 deform_groin_base_r {parent=thigh_r, loc=<-0.4801 -5.6776 0.2486>, rot=<0 0 0 -1>}
 deform_groin_r {parent=deform_groin_base_r, loc=<0 0 0>, rot=<0 0 0 1>}
 deform_glute_base_l {parent=pelvis, loc=<-7.3466 -9.0213 -9.1245>, rot=<0.0172 0.031 -0.4477 -0.8935>}
 deform_glute_l {parent=deform_glute_base_l, loc=<0 0 0>, rot=<0 0 0 1>}
 deform_glute_base_r {parent=pelvis, loc=<-7.3464 -9.0213 9.1245>, rot=<0.031 -0.0172 0.8935 -0.4477>}
 deform_glute_r {parent=deform_glute_base_r, loc=<0 0 0>, rot=<0 0 0 1>}
 base_thigh_rope_l_01 {parent=pelvis, loc=<1.6336 7.6336 -5.0203>, rot=<-0.0178 -0.9913 0.1305 0.0023>}
 dyn_thigh_rope_l_01 {parent=base_thigh_rope_l_01, loc=<0 0 0>, rot=<0 0.0168 0 -0.9999>}
 dyn_thigh_rope_l_02 {parent=dyn_thigh_rope_l_01, loc=<2.1664 0 0>, rot=<0 0 0 1>}
 base_thigh_rope_r_01 {parent=pelvis, loc=<1.6336 7.6336 5.0203>, rot=<-0.0023 -0.1305 0.9913 0.0178>}
 dyn_thigh_rope_r_01 {parent=base_thigh_rope_r_01, loc=<0 0 0>, rot=<0 0.0168 0 -0.9999>}
 dyn_thigh_rope_r_02 {parent=dyn_thigh_rope_r_01, loc=<2.1663 0 0>, rot=<0 0 0 -1>}
 base_thigh_rope_side_l_01 {parent=pelvis, loc=<1.2304 4.2892 -10.568>, rot=<-0.1182 -0.8883 -0.3643 -0.2534>}
 dyn_thigh_rope_side_l_01 {parent=base_thigh_rope_side_l_01, loc=<0 0 0>, rot=<0 0 0 1>}
 dyn_thigh_rope_side_l_02 {parent=dyn_thigh_rope_side_l_01, loc=<1.7039 0 0>, rot=<0 0 0 1>}
 base_thigh_rope_side_r_01 {parent=pelvis, loc=<1.2304 4.2892 10.568>, rot=<-0.1182 -0.8883 0.3643 0.2534>}
 dyn_thigh_rope_side_r_01 {parent=base_thigh_rope_side_r_01, loc=<0 0 0>, rot=<0 0 0 -1>}
 dyn_thigh_rope_side_r_02 {parent=dyn_thigh_rope_side_r_01, loc=<1.7039 0 0>, rot=<0 0 0 -1>}
 attach {parent=root, loc=<0 0 0>, rot=<0 0 0 1>}
 ik_hand_root {parent=root, loc=<0 0 0>, rot=<0 0 0 1>}
 ik_hand_gun {parent=ik_hand_root, loc=<-48.1743 3.3545 104.4381>, rot=<0.6276 0.1309 0.4183 -0.6434>}
 ik_hand_r {parent=ik_hand_gun, loc=<0 0 0>, rot=<0 0 0 1>}
 ik_hand_l {parent=ik_hand_gun, loc=<59.3271 67.6997 34.3524>, rot=<-0.6158 -0.7027 -0.3565 0>}
 ik_foot_root {parent=root, loc=<0 0 0>, rot=<0 0 0 1>}
 ik_foot_l {parent=ik_foot_root, loc=<14.8262 -7.3545 8.296>, rot=<0.0357 0.7045 0.0284 -0.7082>}
 ik_foot_r {parent=ik_foot_root, loc=<-14.8262 -7.3545 8.296>, rot=<0.7082 0.0284 -0.7045 0.0357>}
 root_fp {parent=root, loc=<0 0 0>, rot=<0 0 0 1>}
 attach_fp {parent=root_fp, loc=<0 0 0>, rot=<0 0 0 1>}
 camera_root {parent=root_fp, loc=<0 0 0>, rot=<0 0 0 1>}
 camera {parent=camera_root, loc=<0 0 0>, rot=<0 0 0 1>}
 attach_emote_01 {parent=root, loc=<0 0 0>, rot=<0 0 0 1>}
 attach_emote_02 {parent=root, loc=<0 0 0>, rot=<0 0 0 1>}
 attach_emote_03 {parent=root, loc=<0 0 0>, rot=<0 0 0 1>}
 warp_point_01 {parent=root, loc=<0 0 0>, rot=<0 0 0 1>}
 warp_point_02 {parent=root, loc=<0 0 0>, rot=<0 0 0 1>}
 warp_point_03 {parent=root, loc=<0 0 0>, rot=<0 0 0 1>}
 hand_r_ctrl {controlType=EULER_TRANSFORM, type=Control, loc=<0 0 0>, rot=<0 0 0 1>}
]
