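# Physics asset for the sleeves and sleeve fringes of the female medium FreeCanyon player body, named as its Sprint variant.
# Capsule bodies cover both lower arms, upper and lower sleeve bones, and four two-segment fringe chains per side.
# Sleeve bones are constrained tightly to the lower arms, and fringe roots hang off the lower sleeve with narrow swing limits.
# Fringe tips swing more freely, with wider limits on the left than the right.
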PhysicsAsset /BRCosmetics/Characters/Player/Female/Medium/Bodies/F_MED_FreeCanyon/Meshes/F_MED_FreeCanyon_Sprint_Physics
bodies {
 dyn_slevee_dwn_l [{
   shapeType=Capsule
   length=8.505
   radius=3.005
   shapeName=dyn_slevee_dwn_l_capsule
   center=<7.3081 -0.2692 -1.4554>
   rotation=<rot -0.0003 -94.9997 -90.0001>
  }]
 dyn_slevee_dwn_r [{shapeType=Capsule, length=8.505, radius=3.005, shapeName=dyn_slevee_dwn_r_capsule, center=<7.0658 0 0>, rotation=<rot 0 -85 105>}]
 dyn_slevee_fringes_01_l [{
   shapeType=Capsule
   length=2.505
   radius=1.38
   shapeName=dyn_slevee_fringes_01_l_capsule
   center=<3.5535 -0.127 -0.0122>
   rotation=<rot 0 88.2162 -89.3174>
  }]
 dyn_slevee_fringes_01_r [{
   shapeType=Capsule
   length=3.755
   radius=1.38
   shapeName=dyn_slevee_fringes_01_r_capsule
   center=<4.5782 -0.1262 0.1092>
   rotation=<rot 0 88.2162 90.6826>
  }]
 dyn_slevee_fringes_01a_l [{
   shapeType=Capsule
   length=8.63
   radius=1.38
   shapeName=dyn_slevee_fringes_01_l_capsule
   center=<3.6591 -0.1303 -0.011>
   rotation=<rot 0 88.2162 -89.3174>
  }]
 dyn_slevee_fringes_01a_r [{
   shapeType=Capsule
   length=7.755
   radius=1.38
   shapeName=dyn_slevee_fringes_01_r_capsule
   center=<4.0228 -0.1089 0.1026>
   rotation=<rot 0 88.2162 90.6826>
  }]
 dyn_slevee_fringes_02_l [{
   shapeType=Capsule
   length=2.505
   radius=1.38
   shapeName=dyn_slevee_fringes_01_l_capsule
   center=<3.5535 -0.127 -0.0122>
   rotation=<rot 0 88.2162 -89.3174>
  }]
 dyn_slevee_fringes_02_r [{
   shapeType=Capsule
   length=2.505
   radius=1.38
   shapeName=dyn_slevee_fringes_01_r_capsule
   center=<4.9082 -0.1365 0.1131>
   rotation=<rot 0 88.2162 90.6826>
  }]
 dyn_slevee_fringes_02a_l [{
   shapeType=Capsule
   length=8.13
   radius=1.255
   shapeName=dyn_slevee_fringes_01_l_capsule
   center=<3.5763 -0.1277 -0.012>
   rotation=<rot 0 88.2162 -89.3174>
  }]
 dyn_slevee_fringes_02a_r [{
   shapeType=Capsule
   length=6.505
   radius=1.255
   shapeName=dyn_slevee_fringes_01_r_capsule
   center=<4.1281 -0.1122 0.1038>
   rotation=<rot 0 88.2162 90.6826>
  }]
 dyn_slevee_fringes_03_l [{
   shapeType=Capsule
   length=2.505
   radius=1.38
   shapeName=dyn_slevee_fringes_01_l_capsule
   center=<3.5535 -0.127 -0.0122>
   rotation=<rot 0 88.2162 -89.3174>
  }]
 dyn_slevee_fringes_03_r [{
   shapeType=Capsule
   length=2.505
   radius=1.38
   shapeName=dyn_slevee_fringes_01_r_capsule
   center=<4.7177 -0.1306 0.1108>
   rotation=<rot 0 88.2162 90.6826>
  }]
 dyn_slevee_fringes_03a_l [{
   shapeType=Capsule
   length=7.005
   radius=1.255
   shapeName=dyn_slevee_fringes_01_l_capsule
   center=<3.8835 -0.1373 -0.117>
   rotation=<rot 0 88.2162 -89.3174>
  }]
 dyn_slevee_fringes_03a_r [{
   shapeType=Capsule
   length=5.88
   radius=1.255
   shapeName=dyn_slevee_fringes_01_r_capsule
   center=<4.0055 -0.1084 0.211>
   rotation=<rot 0 88.2162 90.6826>
  }]
 dyn_slevee_fringes_04_l [{
   shapeType=Capsule
   length=2.505
   radius=1.38
   shapeName=dyn_slevee_fringes_01_l_capsule
   center=<3.5535 -0.127 -0.0122>
   rotation=<rot 0 88.2162 -89.3174>
  }]
 dyn_slevee_fringes_04_r [{
   shapeType=Capsule
   length=2.505
   radius=1.38
   shapeName=dyn_slevee_fringes_01_r_capsule
   center=<4.3431 -0.1189 0.1064>
   rotation=<rot 0 88.2162 90.6826>
  }]
 dyn_slevee_fringes_04a_l [{
   shapeType=Capsule
   length=5.13
   radius=1.38
   shapeName=dyn_slevee_fringes_01_l_capsule
   center=<3.4825 -0.1248 -0.0131>
   rotation=<rot 0 88.2162 -89.3174>
  }]
 dyn_slevee_fringes_04a_r [{
   shapeType=Capsule
   length=3.88
   radius=1.38
   shapeName=dyn_slevee_fringes_01_r_capsule
   center=<4.1642 -0.1133 0.1042>
   rotation=<rot 0 88.2162 90.6826>
  }]
 dyn_slevee_uppr_l [{
   shapeType=Capsule
   length=3.63
   radius=3.005
   shapeName=dyn_slevee_dwn_l_capsule
   center=<4.2303 0 -1.4554>
   rotation=<rot -0.0003 -94.9997 -90.0001>
  }]
 dyn_slevee_uppr_r [{shapeType=Capsule, length=3.63, radius=3.005, shapeName=dyn_slevee_dwn_r_capsule, center=<6.1153 0 0>, rotation=<rot 0 -95 80>}]
 lowerarm_l [
  {shapeType=Capsule, length=0.505, radius=0.505, shapeName=lowerarm_l_capsule, rotation=<rot 0 -90 0>}
  {
   shapeType=Capsule
   length=27.75
   radius=8.375
   shapeName=lowerarm_l_capsule
   center=<13.6768 -0.6779 -0.742>
   rotation=<rot -68.0597 179.9992 168.107>
  }
 ]
 lowerarm_r [
  {shapeType=Capsule, length=0.505, radius=0.505, shapeName=lowerarm_r_capsule, rotation=<rot 0 -90 0>}
  {
   shapeType=Capsule
   length=27.75
   radius=8.375
   shapeName=lowerarm_r_capsule
   center=<-13.6768 0.6779 0.742>
   rotation=<rot -68.0597 179.9992 -11.893>
  }
 ]
}
constraints [
 {bone1Name=dyn_slevee_dwn_l, bone2Name=lowerarm_l, swing1LimitDegrees=2, swing2LimitDegrees=2, twistLimitDegrees=0}
 {bone1Name=dyn_slevee_uppr_l, bone2Name=lowerarm_l, swing1LimitDegrees=2, swing2LimitDegrees=2, twistLimitDegrees=0}
 {bone1Name=dyn_slevee_dwn_r, bone2Name=lowerarm_r, swing1LimitDegrees=2, swing2LimitDegrees=2, twistLimitDegrees=0}
 {bone1Name=dyn_slevee_uppr_r, bone2Name=lowerarm_r, swing1LimitDegrees=2, swing2LimitDegrees=2, twistLimitDegrees=0}
 {bone1Name=dyn_slevee_fringes_01_l, bone2Name=dyn_slevee_dwn_l, swing1LimitDegrees=5, swing2LimitDegrees=10, twistLimitDegrees=0}
 {bone1Name=dyn_slevee_fringes_01a_l, bone2Name=dyn_slevee_fringes_01_l, swing1LimitDegrees=45, swing2LimitDegrees=40, twistLimitDegrees=0}
 {bone1Name=dyn_slevee_fringes_02_l, bone2Name=dyn_slevee_dwn_l, swing1LimitDegrees=5, swing2LimitDegrees=10, twistLimitDegrees=0}
 {bone1Name=dyn_slevee_fringes_02a_l, bone2Name=dyn_slevee_fringes_02_l, swing1LimitDegrees=40, swing2LimitDegrees=30, twistLimitDegrees=0}
 {bone1Name=dyn_slevee_fringes_03_l, bone2Name=dyn_slevee_dwn_l, swing1LimitDegrees=5, swing2LimitDegrees=10, twistLimitDegrees=0}
 {bone1Name=dyn_slevee_fringes_03a_l, bone2Name=dyn_slevee_fringes_03_l, swing1LimitDegrees=30, swing2LimitDegrees=25, twistLimitDegrees=0}
 {bone1Name=dyn_slevee_fringes_04_l, bone2Name=dyn_slevee_dwn_l, swing1LimitDegrees=5, swing2LimitDegrees=10, twistLimitDegrees=0}
 {bone1Name=dyn_slevee_fringes_04a_l, bone2Name=dyn_slevee_fringes_04_l, swing1LimitDegrees=30, swing2LimitDegrees=20, twistLimitDegrees=0}
 {
  bone1Name=dyn_slevee_fringes_01_r
  bone2Name=dyn_slevee_dwn_r
  swing1LimitDegrees=5
  swing2LimitDegrees=10
  twistLimitDegrees=45
  twistMotion=Locked
 }
 {bone1Name=dyn_slevee_fringes_01a_r, bone2Name=dyn_slevee_fringes_01_r, swing1LimitDegrees=15, swing2LimitDegrees=20, twistLimitDegrees=0}
 {bone1Name=dyn_slevee_fringes_02_r, bone2Name=dyn_slevee_dwn_r, swing1LimitDegrees=5, swing2LimitDegrees=10, twistLimitDegrees=0}
 {bone1Name=dyn_slevee_fringes_02a_r, bone2Name=dyn_slevee_fringes_02_r, swing1LimitDegrees=15, swing2LimitDegrees=15, twistLimitDegrees=0}
 {bone1Name=dyn_slevee_fringes_03_r, bone2Name=dyn_slevee_dwn_r, swing1LimitDegrees=5, swing2LimitDegrees=10, twistLimitDegrees=0}
 {bone1Name=dyn_slevee_fringes_03a_r, bone2Name=dyn_slevee_fringes_03_r, swing1LimitDegrees=15, swing2LimitDegrees=15, twistLimitDegrees=0}
 {bone1Name=dyn_slevee_fringes_04_r, bone2Name=dyn_slevee_dwn_r, swing1LimitDegrees=5, swing2LimitDegrees=10, twistLimitDegrees=0}
 {bone1Name=dyn_slevee_fringes_04a_r, bone2Name=dyn_slevee_fringes_04_r, swing1LimitDegrees=15, swing2LimitDegrees=15, twistLimitDegrees=0}
]
