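# Physics asset for the skydive parachute mesh of the FreeCanyon female medium cosmetic body, covering its sleeves and sleeve fringes.
# Capsule bodies sit on both forearms, each parenting an upper and a lower sleeve body through constraints with very small angle limits.
# Four two-segment fringe chains hang from each lower sleeve, with wider swing limits and zero twist limit.
# Only forearm, sleeve and fringe bones have bodies.
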PhysicsAsset /BRCosmetics/Characters/Player/Female/Medium/Bodies/F_MED_FreeCanyon/Meshes/F_MED_FreeCanyon_Skydive_Parachute_Physics
bodies {
 dyn_slevee_dwn_l [{
   shapeType=Capsule
   length=7.8377
   radius=3.6429
   shapeName=dyn_slevee_dwn_l_capsule
   center=<6.4646 -1.2946 1.3644>
   rotation=<rot 0 -107.1927 -87.9613>
  }]
 dyn_slevee_dwn_r [{
   shapeType=Capsule
   length=7.8377
   radius=3.6429
   shapeName=dyn_slevee_dwn_r_capsule
   center=<7.5179 0.6158 1.1154>
   rotation=<rot -0.8612 -82.2069 91.8477>
  }]
 dyn_slevee_fringes_01_l [{
   shapeType=Capsule
   length=2.505
   radius=1.38
   shapeName=dyn_slevee_fringes_01_l_capsule
   center=<3.5535 -0.127 -0.0122>
   rotation=<rot 0 88.2162 -89.3174>
  }]
 dyn_slevee_fringes_01_r [{
   shapeType=Capsule
   length=4.255
   radius=1.38
   shapeName=dyn_slevee_fringes_01_r_capsule
   center=<4.5534 -0.1254 0.1089>
   rotation=<rot 0 88.2162 90.6826>
  }]
 dyn_slevee_fringes_01a_l [{
   shapeType=Capsule
   length=8.505
   radius=1.38
   shapeName=dyn_slevee_fringes_01_l_capsule
   center=<3.6718 -0.1307 -0.0108>
   rotation=<rot 0 88.2162 -89.3174>
  }]
 dyn_slevee_fringes_01a_r [{
   shapeType=Capsule
   length=7.38
   radius=1.38
   shapeName=dyn_slevee_fringes_01_r_capsule
   center=<4.348 -0.119 0.1064>
   rotation=<rot 0 88.2162 90.6826>
  }]
 dyn_slevee_fringes_02_l [{
   shapeType=Capsule
   length=3.63
   radius=1.38
   shapeName=dyn_slevee_fringes_01_l_capsule
   center=<4.873 -0.1681 0.0035>
   rotation=<rot 0 88.2162 -89.3174>
  }]
 dyn_slevee_fringes_02_r [{
   shapeType=Capsule
   length=4.505
   radius=1.38
   shapeName=dyn_slevee_fringes_01_r_capsule
   center=<5.5893 -0.1577 0.1212>
   rotation=<rot 0 88.2162 90.6826>
  }]
 dyn_slevee_fringes_02a_l [{
   shapeType=Capsule
   length=6.63
   radius=1.255
   shapeName=dyn_slevee_fringes_01_l_capsule
   center=<4.3053 -0.1504 -0.0033>
   rotation=<rot 0 88.2162 -89.3174>
  }]
 dyn_slevee_fringes_02a_r [{
   shapeType=Capsule
   length=5.13
   radius=1.255
   shapeName=dyn_slevee_fringes_01_r_capsule
   center=<4.9232 -0.137 0.1133>
   rotation=<rot 0 88.2162 90.6826>
  }]
 dyn_slevee_fringes_03_l [{
   shapeType=Capsule
   length=3.13
   radius=1.38
   shapeName=dyn_slevee_fringes_01_l_capsule
   center=<4.3707 -0.1525 -0.0025>
   rotation=<rot 0 88.2162 -89.3174>
  }]
 dyn_slevee_fringes_03_r [{
   shapeType=Capsule
   length=2.88
   radius=1.38
   shapeName=dyn_slevee_fringes_01_r_capsule
   center=<4.9082 -0.1365 0.1131>
   rotation=<rot 0 88.2162 90.6826>
  }]
 dyn_slevee_fringes_03a_l [{
   shapeType=Capsule
   length=6.255
   radius=1.255
   shapeName=dyn_slevee_fringes_01_l_capsule
   center=<3.966 -0.1399 -0.116>
   rotation=<rot 0 88.2162 -89.3174>
  }]
 dyn_slevee_fringes_03a_r [{
   shapeType=Capsule
   length=4.755
   radius=1.13
   shapeName=dyn_slevee_fringes_01_r_capsule
   center=<3.9089 -0.1054 0.2099>
   rotation=<rot 0 88.2162 90.6826>
  }]
 dyn_slevee_fringes_04_l [{
   shapeType=Capsule
   length=2.63
   radius=1.255
   shapeName=dyn_slevee_fringes_01_l_capsule
   center=<3.5535 -0.127 -0.0122>
   rotation=<rot 0 88.2162 -89.3174>
  }]
 dyn_slevee_fringes_04_r [{
   shapeType=Capsule
   length=2.63
   radius=1.255
   shapeName=dyn_slevee_fringes_01_r_capsule
   center=<4.5367 -0.1249 0.1087>
   rotation=<rot 0 88.2162 90.6826>
  }]
 dyn_slevee_fringes_04a_l [{
   shapeType=Capsule
   length=5.255
   radius=1.38
   shapeName=dyn_slevee_fringes_01_l_capsule
   center=<3.5161 -0.1259 -0.0127>
   rotation=<rot 0 88.2162 -89.3174>
  }]
 dyn_slevee_fringes_04a_r [{
   shapeType=Capsule
   length=3.755
   radius=1.38
   shapeName=dyn_slevee_fringes_01_r_capsule
   center=<4.1642 -0.1133 0.1042>
   rotation=<rot 0 88.2162 90.6826>
  }]
 dyn_slevee_uppr_l [{
   shapeType=Capsule
   length=9.1812
   radius=3.1926
   shapeName=dyn_slevee_uppr_l_capsule
   center=<8.1484 -1.7583 0.0085>
   rotation=<rot -3.6705 -89.9325 -96.9148>
  }]
 dyn_slevee_uppr_r [{
   shapeType=Capsule
   length=9.1812
   radius=3.1926
   shapeName=dyn_slevee_uppr_r_capsule
   center=<6.9729 1.8075 -0.4965>
   rotation=<rot -3.6702 -89.9327 78.0851>
  }]
 lowerarm_l [
  {shapeType=Capsule, length=0.505, radius=0.505, shapeName=lowerarm_l_capsule, rotation=<rot 0 -90 0>}
  {
   shapeType=Capsule
   length=27.75
   radius=8.375
   shapeName=lowerarm_l_capsule
   center=<13.6768 -0.6779 -0.742>
   rotation=<rot -68.0597 179.9992 168.107>
  }
 ]
 lowerarm_r [
  {shapeType=Capsule, length=0.505, radius=0.505, shapeName=lowerarm_r_capsule, rotation=<rot 0 -90 180>}
  {
   shapeType=Capsule
   length=27.75
   radius=8.375
   shapeName=lowerarm_r_capsule
   center=<-13.6768 0.6779 0.742>
   rotation=<rot -68.0597 179.9992 -11.893>
  }
 ]
}
constraints [
 {bone1Name=dyn_slevee_dwn_l, bone2Name=lowerarm_l, swing1LimitDegrees=0.5, swing2LimitDegrees=1, twistLimitDegrees=0}
 {bone1Name=dyn_slevee_uppr_l, bone2Name=lowerarm_l, swing1LimitDegrees=1, swing2LimitDegrees=2, twistLimitDegrees=1}
 {bone1Name=dyn_slevee_dwn_r, bone2Name=lowerarm_r, swing1LimitDegrees=0.5, swing2LimitDegrees=1, twistLimitDegrees=0}
 {bone1Name=dyn_slevee_uppr_r, bone2Name=lowerarm_r, swing1LimitDegrees=1, swing2LimitDegrees=2, twistLimitDegrees=1}
 {bone1Name=dyn_slevee_fringes_01_l, bone2Name=dyn_slevee_dwn_l, swing1LimitDegrees=5, swing2LimitDegrees=45, twistLimitDegrees=0}
 {bone1Name=dyn_slevee_fringes_01a_l, bone2Name=dyn_slevee_fringes_01_l, swing1LimitDegrees=15, swing2LimitDegrees=15, twistLimitDegrees=0}
 {bone1Name=dyn_slevee_fringes_02_l, bone2Name=dyn_slevee_dwn_l, swing1LimitDegrees=5, swing2LimitDegrees=45, twistLimitDegrees=0}
 {bone1Name=dyn_slevee_fringes_02a_l, bone2Name=dyn_slevee_fringes_02_l, swing1LimitDegrees=15, swing2LimitDegrees=15, twistLimitDegrees=0}
 {bone1Name=dyn_slevee_fringes_03_l, bone2Name=dyn_slevee_dwn_l, swing1LimitDegrees=5, swing2LimitDegrees=45, twistLimitDegrees=0}
 {bone1Name=dyn_slevee_fringes_03a_l, bone2Name=dyn_slevee_fringes_03_l, swing1LimitDegrees=15, swing2LimitDegrees=15, twistLimitDegrees=0}
 {bone1Name=dyn_slevee_fringes_04_l, bone2Name=dyn_slevee_dwn_l, swing1LimitDegrees=5, swing2LimitDegrees=45, twistLimitDegrees=0}
 {bone1Name=dyn_slevee_fringes_04a_l, bone2Name=dyn_slevee_fringes_04_l, swing1LimitDegrees=15, swing2LimitDegrees=15, twistLimitDegrees=0}
 {bone1Name=dyn_slevee_fringes_01_r, bone2Name=dyn_slevee_dwn_r, swing1LimitDegrees=5, swing2LimitDegrees=45, twistLimitDegrees=0}
 {bone1Name=dyn_slevee_fringes_01a_r, bone2Name=dyn_slevee_fringes_01_r, swing1LimitDegrees=15, swing2LimitDegrees=15, twistLimitDegrees=0}
 {bone1Name=dyn_slevee_fringes_02_r, bone2Name=dyn_slevee_dwn_r, swing1LimitDegrees=5, swing2LimitDegrees=45, twistLimitDegrees=0}
 {bone1Name=dyn_slevee_fringes_02a_r, bone2Name=dyn_slevee_fringes_02_r, swing1LimitDegrees=15, swing2LimitDegrees=15, twistLimitDegrees=0}
 {bone1Name=dyn_slevee_fringes_03_r, bone2Name=dyn_slevee_dwn_r, swing1LimitDegrees=5, swing2LimitDegrees=45, twistLimitDegrees=0}
 {bone1Name=dyn_slevee_fringes_03a_r, bone2Name=dyn_slevee_fringes_03_r, swing1LimitDegrees=15, swing2LimitDegrees=15, twistLimitDegrees=0}
 {bone1Name=dyn_slevee_fringes_04_r, bone2Name=dyn_slevee_dwn_r, swing1LimitDegrees=5, swing2LimitDegrees=45, twistLimitDegrees=0}
 {bone1Name=dyn_slevee_fringes_04a_r, bone2Name=dyn_slevee_fringes_04_r, swing1LimitDegrees=15, swing2LimitDegrees=15, twistLimitDegrees=0}
]
