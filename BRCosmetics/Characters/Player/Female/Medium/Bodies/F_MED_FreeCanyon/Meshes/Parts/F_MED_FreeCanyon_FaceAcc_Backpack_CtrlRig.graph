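# Control Rig for the F_MED_FreeCanyon face accessory that repositions lower hair bones, apparently for backpack (back bling) wear.
# Four chained Parent Constraints blend dyn_hair_lwr_01–04 between spine_05 and either head or base_hair_master_01, favoring the spine.
# Constraints for the "a" child bones and a Modify Transforms node sit outside the execution chain.
# isWearingBackbling feeds bool-to-float nodes whose outputs go unused, so constraint weights stay fixed.
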ControlRigBlueprint /BRCosmetics/Characters/Player/Female/Medium/Bodies/F_MED_FreeCanyon/Meshes/Parts/F_MED_FreeCanyon_FaceAcc_Backpack_CtrlRig
graphs [
 RigVMFunctionLibrary {links [], nodes []}
 RigVMModel {nodes [
   RigUnit_BeginExecution = &RigUnit_BeginExecution()
   &RigUnit_ParentConstraint(
    AdvancedSettings=(InterpolationType=Average,RotationOrderForFilter=XZY)
    Child=(Type=Bone,Name="dyn_hair_lwr_01")
    Parents=((Item=(Type=Bone,Name="spine_05"),Weight=0.6),(Item=(Type=Bone,Name="head"),Weight=0.4))
    Weight=1
    ExecutePin=@RigUnit_BeginExecution.ExecutePin
   )
   > &RigUnit_ParentConstraint(
    AdvancedSettings=(InterpolationType=Average,RotationOrderForFilter=XZY)
    Child=(Type=Bone,Name="dyn_hair_lwr_02")
    Parents=((Item=(Type=Bone,Name="spine_05"),Weight=0.6),(Item=(Type=Bone,Name="head"),Weight=0.4))
    Weight=1
   )
   > &RigUnit_ParentConstraint(
    AdvancedSettings=(InterpolationType=Average,RotationOrderForFilter=XZY)
    Child=(Type=Bone,Name="dyn_hair_lwr_03")
    Parents=((Item=(Type=Bone,Name="spine_05"),Weight=0.6),(Item=(Type=Bone,Name="base_hair_master_01"),Weight=0.4))
    Weight=1
   )
   > &RigUnit_ParentConstraint(
    AdvancedSettings=(InterpolationType=Average,RotationOrderForFilter=XZY)
    Child=(Type=Bone,Name="dyn_hair_lwr_04")
    Parents=((Item=(Type=Bone,Name="spine_05"),Weight=0.6),(Item=(Type=Bone,Name="base_hair_master_01"),Weight=0.4))
    Weight=1
   )
   &RigUnit_ModifyTransforms(
    ExecutePin=()
    ItemToModify=((Item=(Type=Bone,Name="dyn_hair_lwr_01"),Transform=(Rotation=(X=-0.25,Y=-0.433,Z=-0.433,W=0.75),Translation=(X=100,Y=0,Z=0),Scale3D=(X=1,Y=1,Z=1))))
    Weight=1
    WeightMinimum=0
   )
   &RigUnit_ParentConstraint(
    AdvancedSettings=(InterpolationType=Average,RotationOrderForFilter=XZY)
    Child=(Type=Bone,Name="dyn_hair_lwr_01a")
    ExecutePin=()
    Parents=((Item=(Type=Bone,Name="spine_05"),Weight=0.6),(Item=(Type=Bone,Name="head"),Weight=0.4))
    Weight=1
   )
   &RigUnit_ParentConstraint(
    AdvancedSettings=(InterpolationType=Average,RotationOrderForFilter=XZY)
    Child=(Type=Bone,Name="dyn_hair_lwr_02a")
    ExecutePin=()
    Parents=((Item=(Type=Bone,Name="spine_05"),Weight=0.6),(Item=(Type=Bone,Name="base_hair_master_01"),Weight=0.4))
    Weight=1
   )
   &RigUnit_ParentConstraint(
    AdvancedSettings=(InterpolationType=Average,RotationOrderForFilter=XZY)
    Child=(Type=Bone,Name="dyn_hair_lwr_03a")
    ExecutePin=()
    Parents=((Item=(Type=Bone,Name="spine_05"),Weight=0.6),(Item=(Type=Bone,Name="base_hair_master_01"),Weight=0.4))
    Weight=1
   )
   &RigUnit_ParentConstraint(
    AdvancedSettings=(InterpolationType=Average,RotationOrderForFilter=XZY)
    Child=(Type=Bone,Name="dyn_hair_lwr_04a")
    ExecutePin=()
    Parents=((Item=(Type=Bone,Name="spine_05"),Weight=0.6),(Item=(Type=Bone,Name="base_hair_master_01"),Weight=0.4))
    Weight=1
   )
   RigVMFunction_MathBoolToFloat = &RigVMFunction_MathBoolToFloat(Value=RigVMVariableNode() {variable=isWearingBackbling}.Value)
   &RigVMFunction_MathBoolToFloat(Value=RigVMVariableNode() {variable=isWearingBackbling}.Value)
  ]}
]
hierarchy [
 root {loc=<0 0 0>, rot=<0 0 0 1>}
 pelvis {parent=root, loc=<0 -0.5301 98.0571>, rot=<0.0127 0.707 -0.0127 -0.707>}
 spine_01 {parent=pelvis, loc=<6.1208 -0.9328 0>, rot=<0 0 0.034 -0.9994>}
 spine_02 {parent=spine_01, loc=<6.5997 -0.2095 0>, rot=<0 0 0.0238 -0.9997>}
 spine_03 {parent=spine_02, loc=<6.4914 -0.1237 0>, rot=<0 0 0.0149 -0.9999>}
 spine_04 {parent=spine_03, loc=<6.9598 0.0794 0>, rot=<0 0 -0.0132 -0.9999>}
 spine_05 {parent=spine_04, loc=<6.9987 0 0>, rot=<0 0 -0.0015 -1>}
 neck_01 {parent=spine_05, loc=<11.3138 0.1529 0>, rot=<0 0 -0.2003 -0.9797>}
 neck_02 {parent=neck_01, loc=<4.5627 0 0>, rot=<0 0 0.0504 -0.9987>}
 head {parent=neck_02, loc=<5.3914 0 0>, rot=<0 0 0.115 -0.9934>}
 base_hair_master_01 {parent=head, loc=<5.0578 -9.1925 0>, rot=<0.5021 -0.4979 0.4979 -0.5021>}
 dyn_hair_master_01 {parent=base_hair_master_01, loc=<0 0 0>, rot=<0 0 0 1>}
 dyn_hair_uppr_01 {parent=dyn_hair_master_01, loc=<4.5721 -3.1136 7.9393>, rot=<0.099 0.2509 0.2426 -0.9319>}
 dyn_hair_uppr_01a {parent=dyn_hair_uppr_01, loc=<7.3863 -0.4727 -0.9964>, rot=<0 -0.1295 0 -0.9916>}
 dyn_hair_uppr_02 {parent=dyn_hair_master_01, loc=<6.1846 -0.6349 6.1409>, rot=<0.1426 0.2025 0.0671 -0.9665>}
 dyn_hair_uppr_02a {parent=dyn_hair_uppr_02, loc=<6.6824 -0.4723 -0.8002>, rot=<0.0008 -0.1727 0.002 -0.985>}
 dyn_hair_uppr_03 {parent=dyn_hair_master_01, loc=<6.2871 -8.8626 7.0273>, rot=<-0.1645 0.1819 0.2001 -0.9486>}
 dyn_hair_uppr_03a {parent=dyn_hair_uppr_03, loc=<4.2696 -0.9226 -0.1371>, rot=<0.0034 -0.0285 0.0819 -0.9962>}
 dyn_hair_uppr_04 {parent=dyn_hair_master_01, loc=<0.3412 -7.1109 3.8601>, rot=<-0.1134 0.2461 0.5987 -0.7538>}
 dyn_hair_uppr_05 {parent=dyn_hair_master_01, loc=<-0.6943 -7.2065 -2.6289>, rot=<-0.52 0.1583 0.7096 -0.4484>}
 dyn_hair_uppr_06 {parent=dyn_hair_master_01, loc=<6.1104 7.7455 7.3942>, rot=<0.18 0.24 -0.365 -0.8813>}
 dyn_hair_uppr_06a {parent=dyn_hair_uppr_06, loc=<5.6753 -0.3817 -1.2827>, rot=<0.0136 -0.2241 0.1088 -0.9684>}
 dyn_hair_mid_02 {parent=dyn_hair_master_01, loc=<0.9272 9.4967 -6.7903>, rot=<-0.1529 -0.6937 -0.3843 -0.5896>}
 dyn_hair_lwr_01 {parent=dyn_hair_master_01, loc=<1.8344 4.5261 -8.9563>, rot=<-0.4595 -0.6272 0.1984 -0.5968>}
 dyn_hair_lwr_01a {parent=dyn_hair_lwr_01, loc=<9.5517 0.5094 -0.7898>, rot=<-0.0493 -0.2217 -0.055 -0.9723>}
 dyn_hair_lwr_02 {parent=dyn_hair_master_01, loc=<2.1346 1.4692 -12.2591>, rot=<0.0132 -0.4298 -0.0753 -0.8997>}
 dyn_hair_lwr_02a {parent=dyn_hair_lwr_02, loc=<8.121 0.3549 -0.5314>, rot=<-0.0581 -0.024 0.0137 -0.9979>}
 dyn_hair_lwr_03 {parent=dyn_hair_master_01, loc=<2.4775 -5.8475 -11.6577>, rot=<0.1055 -0.4705 0.1471 -0.8636>}
 dyn_hair_lwr_03a {parent=dyn_hair_lwr_03, loc=<5.8784 0.2758 -0.7989>, rot=<-0.0404 0.0868 -0.0877 -0.9915>}
 dyn_hair_lwr_04 {parent=dyn_hair_master_01, loc=<0.4661 -6.0386 -6.5027>, rot=<-0.0023 -0.5114 0.3809 -0.7703>}
 dyn_hair_lwr_04a {parent=dyn_hair_lwr_04, loc=<4.7127 0.2917 -0.0786>, rot=<0.0082 -0.1976 0.0143 -0.9802>}
 dyn_hair_mid_03 {parent=dyn_hair_master_01, loc=<6.0891 -12.9006 -3.4561>, rot=<-0.3455 -0.2389 0.3544 -0.8354>}
 dyn_hair_mid_03a {parent=dyn_hair_mid_03, loc=<4.3877 -0.7257 0.1937>, rot=<0.0085 -0.2244 0.0867 -0.9706>}
 dyn_hair_mid_04 {parent=dyn_hair_master_01, loc=<10.4369 -8.9846 -10.2644>, rot=<-0.0954 -0.4297 0.2539 -0.8613>}
 dyn_hair_uppr_07 {parent=dyn_hair_master_01, loc=<0.455 5.5239 3.6425>, rot=<0.0421 0.1511 -0.5785 -0.8005>}
 dyn_hair_mid_01 {parent=dyn_hair_master_01, loc=<6.528 -2.6005 -0.4843>, rot=<0.0311 -0.1058 0.0117 -0.9938>}
 dyn_hair_mid_01a {parent=dyn_hair_mid_01, loc=<6.1795 -0.4361 -1.5555>, rot=<-0.0173 -0.2239 -0.0249 -0.9741>}
 dyn_hair_mid_05 {parent=dyn_hair_master_01, loc=<8.0687 4.7022 -0.9745>, rot=<0.0095 -0.1356 -0.2416 -0.9608>}
 dyn_hair_mid_05a {parent=dyn_hair_mid_05, loc=<5.5721 -0.4742 -1.2612>, rot=<-0.0194 -0.1362 -0.0233 -0.9902>}
 base_headband_master_01 {parent=head, loc=<-1.1823 -7.1187 -0.0103>, rot=<0.5021 -0.4979 0.4979 -0.5021>}
 dyn_headband_master_01 {parent=base_headband_master_01, loc=<0 0 0>, rot=<0 0 0 1>}
 dyn_headband_l_01 {parent=dyn_headband_master_01, loc=<0.1095 1.291 -1.0359>, rot=<0.0585 -0.6138 -0.3319 -0.7139>}
 dyn_headband_l_02 {parent=dyn_headband_l_01, loc=<2.8192 0.0889 -0.2302>, rot=<0.0139 -0.1148 0.0041 -0.9933>}
 dyn_headband_r_01 {parent=dyn_headband_master_01, loc=<0.1095 -1.291 -1.0361>, rot=<0.7139 -0.3319 -0.6138 -0.0585>}
 dyn_headband_r_02 {parent=dyn_headband_r_01, loc=<2.8187 0.0891 0.23>, rot=<-0.0139 0.1148 0.0041 -0.9933>}
 base_eyering_l_01 {parent=head, loc=<0.2151 0.2353 -6.1801>, rot=<-0.0256 -0.9597 -0.2576 -0.1093>}
 dyn_eyering_l_01 {parent=base_eyering_l_01, loc=<0 0 0>, rot=<0 0 0 1>}
 dyn_eyering_l_02 {parent=dyn_eyering_l_01, loc=<2.3454 -0.0689 0.0685>, rot=<0 -0.1033 0 -0.9947>}
 base_eyering_r_01 {parent=head, loc=<0.2148 0.2353 6.1801>, rot=<-0.0256 -0.9597 0.2576 0.1093>}
 dyn_eyering_r_01 {parent=base_eyering_r_01, loc=<0 0 0>, rot=<0 0 0 1>}
 dyn_eyering_r_02 {parent=dyn_eyering_r_01, loc=<2.3448 -0.0689 -0.0687>, rot=<0 0.1033 0 -0.9947>}
 clavicle_l {parent=spine_05, loc=<5.4831 -0.1053 -3.2946>, rot=<0.0465 -0.7289 -0.009 -0.6829>}
 upperarm_l {parent=clavicle_l, loc=<9.8658 -1.2136 -0.4105>, rot=<0.0164 -0.3245 -0.0269 -0.9454>}
 upperarm_twist_01_l {parent=upperarm_l, loc=<7.2135 -0.1134 0>, rot=<0 0 0 1>}
 clavicle_r {parent=spine_05, loc=<5.4831 -0.1053 3.2946>, rot=<0.7289 0.0465 -0.6829 0.009>}
 upperarm_r {parent=clavicle_r, loc=<-9.8658 1.2136 0.4105>, rot=<0.0164 -0.3245 -0.0269 -0.9454>}
 upperarm_twist_01_r {parent=upperarm_r, loc=<-7.2135 0.1134 0>, rot=<0 0 0 -1>}
]
variables [isWearingBackbling {type=bool, default=False}]
