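# Physics asset for the FreeCanyon medium female face accessory, providing bodies and joints for its dynamic hair, headband and eye rings.
# Capsules cover those bones, neck, clavicles and upper spine, with a sphere for the head.
# Hair strands branch from a nearly rigid master on the head; headband and eye-ring chains have small swing and twist limits.
# Body joints, the headband master and dyn_hair_lwr_04 have free angular motion, unlike the other constraints.
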PhysicsAsset /BRCosmetics/Characters/Player/Female/Medium/Bodies/F_MED_FreeCanyon/Meshes/Parts/F_MED_FreeCanyon_FaceAcc_Physics
bodies {
 clavicle_l [{
   shapeType=Capsule
   length=32.6902
   radius=7.2949
   shapeName=clavicle_l_capsule
   center=<6.2026 -3.7654 -3.2533>
   rotation=<rot 35.7178 97.3221 -88.136>
  }]
 clavicle_r [{
   shapeType=Capsule
   length=32.6902
   radius=7.2949
   shapeName=clavicle_r_capsule
   center=<-6.2026 3.7654 3.2533>
   rotation=<rot 35.7178 97.3221 91.864>
  }]
 dyn_eyering_l_01 [{
   shapeType=Capsule
   length=1.2953
   radius=1.2951
   shapeName=dyn_eyering_l_01_capsule
   center=<1.5792 -0.0233 -0.1023>
   rotation=<rot -0.538 -89.9998 -85.0127>
  }]
 dyn_eyering_l_02 [{
   shapeType=Capsule
   length=2.505
   radius=2.255
   shapeName=dyn_eyering_l_02_capsule
   center=<3.4953 -0.0347 0.4726>
   rotation=<rot -1.3117 -89.9997 101.6399>
  }]
 dyn_eyering_r_01 [{
   shapeType=Capsule
   length=1.2953
   radius=1.2951
   shapeName=dyn_eyering_r_01_capsule
   center=<1.0004 0.0255 -0.1228>
   rotation=<rot -0.538 -89.9998 94.9873>
  }]
 dyn_eyering_r_02 [{
   shapeType=Capsule
   length=2.505
   radius=2.255
   shapeName=dyn_eyering_r_02_capsule
   center=<3.6577 -0.1549 -0.4568>
   rotation=<rot -1.3117 -89.9995 -98.36>
  }]
 dyn_hair_lwr_01 [
  {
   shapeType=Capsule
   length=3.1288
   radius=2.3942
   shapeName=dyn_hair_lwr_01_capsule
   center=<3.3552 0.4084 -0.2271>
   rotation=<rot 10.7137 -86.6062 -95.0613>
  }
  {
   shapeType=Capsule
   length=3.875
   radius=2.375
   shapeName=dyn_hair_lwr_01_capsule
   center=<3.9739 3.8672 -0.9128>
   rotation=<rot 80.0092 2.5075 -2.5109>
  }
 ]
 dyn_hair_lwr_01a [
  {
   shapeType=Capsule
   length=1.8788
   radius=1.8942
   shapeName=dyn_hair_lwr_01_capsule
   center=<-0.0655 0.5041 -0.0772>
   rotation=<rot 13.3179 -71.9052 -96.9468>
  }
  {
   shapeType=Capsule
   length=3.625
   radius=2
   shapeName=dyn_hair_lwr_01a_capsule
   center=<0.3762 4.782 -0.6042>
   rotation=<rot -80.0004 0.0041 16.6986>
  }
 ]
 dyn_hair_lwr_02 [
  {
   shapeType=Capsule
   length=4.6288
   radius=2.3942
   shapeName=dyn_hair_lwr_01_capsule
   center=<5.8736 -1.9303 0.3851>
   rotation=<rot 10.7136 -86.6062 -100.0612>
  }
  {
   shapeType=Capsule
   length=4
   radius=2.375
   shapeName=dyn_hair_lwr_01_capsule
   center=<2.302 1.4595 0.0167>
   rotation=<rot 86.2152 -1.6185 -6.6131>
  }
 ]
 dyn_hair_lwr_02a [
  {
   shapeType=Capsule
   length=3.3788
   radius=1.8942
   shapeName=dyn_hair_lwr_01_capsule
   center=<0.5656 -2.1663 1.2157>
   rotation=<rot 11.9722 -81.671 -109.0893>
  }
  {
   shapeType=Capsule
   length=9.375
   radius=2
   shapeName=dyn_hair_lwr_01a_capsule
   center=<0.0997 2.3809 0.8226>
   rotation=<rot -75.204 6.7057 13.4438>
  }
 ]
 dyn_hair_lwr_03 [
  {
   shapeType=Capsule
   length=4.8788
   radius=2.8942
   shapeName=dyn_hair_lwr_01_capsule
   center=<3.5976 -2.1585 -0.1072>
   rotation=<rot 10.7139 -86.6061 -70.0606>
  }
  {
   shapeType=Capsule
   length=3.125
   radius=3.125
   shapeName=dyn_hair_lwr_01_capsule
   center=<0.5746 1.3034 0.7479>
   rotation=<rot 80.0092 2.5075 -2.5109>
  }
 ]
 dyn_hair_lwr_03a [
  {
   shapeType=Capsule
   length=3.3788
   radius=1.8942
   shapeName=dyn_hair_lwr_01_capsule
   center=<-0.5961 -4.6727 0.1268>
   rotation=<rot 11.9724 -81.6708 -54.0883>
  }
  {
   shapeType=Capsule
   length=4.875
   radius=2.5
   shapeName=dyn_hair_lwr_01a_capsule
   center=<0.7749 2.1571 -0.2785>
   rotation=<rot -58.8084 -154.961 153.6898>
  }
 ]
 dyn_hair_lwr_04 [
  {
   shapeType=Capsule
   length=3.3788
   radius=1.8942
   shapeName=dyn_hair_lwr_01_capsule
   center=<-0.5961 -4.6727 0.1268>
   rotation=<rot 11.9724 -81.6708 -54.0883>
  }
  {
   shapeType=Capsule
   length=7.625
   radius=1.75
   shapeName=dyn_hair_lwr_01a_capsule
   center=<5.8389 -0.2293 -1.1281>
   rotation=<rot -74.2877 -120.9197 122.0423>
  }
 ]
 dyn_hair_master_01 [{shapeType=Capsule, length=1, radius=1, shapeName=dyn_hair_master_01_capsule}]
 dyn_hair_mid_01 [{
   shapeType=Capsule
   length=9.7854
   radius=6.7011
   shapeName=dyn_hair_mid_05_capsule
   center=<0.7307 2.1086 -0.676>
   rotation=<rot -2.3256 -93.3138 2.6656>
  }]
 dyn_hair_mid_01a [{
   shapeType=Capsule
   length=11.6604
   radius=7.0761
   shapeName=dyn_hair_mid_05_capsule
   center=<2.9266 2.03 -1.8679>
   rotation=<rot -2.3255 -93.3138 -12.3345>
  }]
 dyn_hair_mid_02 [{
   shapeType=Capsule
   length=8.47
   radius=3.5475
   shapeName=dyn_hair_mid_02_capsule
   center=<3.2822 -0.7855 3.4389>
   rotation=<rot 37.7031 -115.7409 -104.3645>
  }]
 dyn_hair_mid_03 [{
   shapeType=Capsule
   length=3.9733
   radius=3.6214
   shapeName=dyn_hair_mid_03_capsule
   center=<0.6154 0.3628 0.1309>
   rotation=<rot -19.605 -90.0001 78.8958>
  }]
 dyn_hair_mid_03a [{
   shapeType=Capsule
   length=4.0983
   radius=3.7464
   shapeName=dyn_hair_mid_03_capsule
   center=<0.288 1.5913 1.57>
   rotation=<rot -14.2368 -81.1771 56.3205>
  }]
 dyn_hair_mid_04 [{
   shapeType=Capsule
   length=6.057
   radius=5.5645
   shapeName=dyn_hair_mid_04_capsule
   center=<2.7454 0.5484 1.0206>
   rotation=<rot 0 100.0632 -87.4568>
  }]
 dyn_hair_mid_05 [{
   shapeType=Capsule
   length=9.7854
   radius=6.7011
   shapeName=dyn_hair_mid_05_capsule
   center=<0.7307 2.1086 -0.676>
   rotation=<rot -2.3256 -93.3138 2.6656>
  }]
 dyn_hair_mid_05a [{
   shapeType=Capsule
   length=11.6604
   radius=7.0761
   shapeName=dyn_hair_mid_05_capsule
   center=<2.9266 2.03 -1.8679>
   rotation=<rot -2.3255 -93.3138 -12.3345>
  }]
 dyn_hair_uppr_01 [{
   shapeType=Capsule
   length=2.8583
   radius=2.1447
   shapeName=dyn_hair_uppr_01_capsule
   center=<1.9139 0.0861 0.0078>
   rotation=<rot 0 -94.7521 -80.0371>
  }]
 dyn_hair_uppr_01a [{
   shapeType=Capsule
   length=2.0351
   radius=1.628
   shapeName=dyn_hair_uppr_01a_capsule
   center=<1.7599 -0.2657 -0.0239>
   rotation=<rot 0 -98.5865 -88.8648>
  }]
 dyn_hair_uppr_02 [{
   shapeType=Capsule
   length=3.2333
   radius=2.5197
   shapeName=dyn_hair_uppr_01_capsule
   center=<1.9139 0.0861 0.0078>
   rotation=<rot 0 -94.7521 -80.0371>
  }]
 dyn_hair_uppr_02a [{
   shapeType=Capsule
   length=4.1601
   radius=2.503
   shapeName=dyn_hair_uppr_01a_capsule
   center=<1.7599 -0.2657 -0.0239>
   rotation=<rot 0.0001 -98.5864 -78.8646>
  }]
 dyn_hair_uppr_03 [{
   shapeType=Capsule
   length=4.8206
   radius=3.7019
   shapeName=dyn_hair_uppr_03_capsule
   center=<0.6889 -0.2579 -0.8292>
   rotation=<rot 13.0048 68.8734 -83.8175>
  }]
 dyn_hair_uppr_03a [{
   shapeType=Capsule
   length=6.2758
   radius=2.31
   shapeName=dyn_hair_uppr_03a_capsule
   center=<1.9696 2.2796 -0.0977>
   rotation=<rot 3.0986 86.9521 -101.8969>
  }]
 dyn_hair_uppr_04 [{
   shapeType=Capsule
   length=4.9684
   radius=4.2862
   shapeName=dyn_hair_uppr_04_capsule
   center=<3.3018 -2.2035 -1.8723>
   rotation=<rot -22.6394 -126.7398 -65.3502>
  }]
 dyn_hair_uppr_05 [{
   shapeType=Capsule
   length=8.5823
   radius=3.7226
   shapeName=dyn_hair_uppr_05_capsule
   center=<4.9022 3.3792 -2.6853>
   rotation=<rot 1.6412 -91.5602 -72.6555>
  }]
 dyn_hair_uppr_06 [{
   shapeType=Capsule
   length=3.9927
   radius=3.3943
   shapeName=dyn_hair_uppr_06_capsule
   center=<1.4125 -0.4523 -1.7479>
   rotation=<rot -43.8695 89.9148 -95.1758>
  }]
 dyn_hair_uppr_06a [{
   shapeType=Capsule
   length=5.2427
   radius=2.1443
   shapeName=dyn_hair_uppr_06_capsule
   center=<1.9065 0.326 -0.6614>
   rotation=<rot -42.2184 117.3141 -109.0711>
  }]
 dyn_hair_uppr_07 [{
   shapeType=Capsule
   length=5.2724
   radius=3.391
   shapeName=dyn_hair_uppr_07_capsule
   center=<4.8879 2.4603 -2.4329>
   rotation=<rot -36.5273 -107.7579 -54.8452>
  }]
 dyn_headband_l_01 [{
   shapeType=Capsule
   length=0.88
   radius=1.005
   shapeName=dyn_headband_l_01_capsule
   center=<1.1996 0.0254 0.0006>
   rotation=<rot -4.8462 -89.9999 -90.501>
  }]
 dyn_headband_l_02 [{
   shapeType=Capsule
   length=1.1098
   radius=1.0715
   shapeName=dyn_headband_l_02_capsule
   center=<0.7776 0.0221 0.0106>
   rotation=<rot 0 -88.0062 -75.6501>
  }]
 dyn_headband_master_01 [{
   shapeType=Capsule
   length=1.588
   radius=1.0461
   shapeName=dyn_headband_master_01_capsule
   center=<0.0082 0.2198 -0.0975>
   rotation=<rot 0 -2.5208 -81.1797>
  }]
 dyn_headband_r_01 [{
   shapeType=Capsule
   length=0.88
   radius=1.005
   shapeName=dyn_headband_l_01_capsule
   center=<1.1996 0.0254 0.0006>
   rotation=<rot -4.8462 -89.9999 -90.501>
  }]
 dyn_headband_r_02 [{
   shapeType=Capsule
   length=1.1098
   radius=1.0715
   shapeName=dyn_headband_l_02_capsule
   center=<0.7776 0.0221 0.0106>
   rotation=<rot 0 -88.0062 -75.6501>
  }]
 head [{shapeType=Sphere, radius=5.8539, shapeName=head_sphere, center=<6.0624 -0.6281 1.2682>}]
 neck_01 [{shapeType=Capsule, length=11, radius=4.38, shapeName=neck_01_capsule, center=<0.9905 0.2083 0>, rotation=<rot 0 -90.0006 -90>}]
 neck_02 [{shapeType=Capsule, length=3.255, radius=4.505, shapeName=neck_02_capsule, rotation=<rot 0 -90.0001 90>}]
 spine_05 [
  {shapeType=Capsule, length=39.63, radius=10.255, shapeName=spine_05_capsule, center=<-4.9482 0.664 0>, rotation=<rot 0 -90 0>}
  {shapeType=Capsule, length=19.5, radius=6.875, shapeName=spine_05_capsule, center=<9.3165 4.1632 0>}
 ]
}
constraints [
 {bone1Name=dyn_hair_lwr_01, bone2Name=dyn_hair_master_01, swing1LimitDegrees=5, swing2LimitDegrees=5, twistLimitDegrees=0}
 {bone1Name=dyn_hair_lwr_01a, bone2Name=dyn_hair_lwr_01, swing1LimitDegrees=5, swing2LimitDegrees=5, twistLimitDegrees=0}
 {bone1Name=dyn_hair_master_01, bone2Name=head, swing1LimitDegrees=0, swing2LimitDegrees=1, twistLimitDegrees=0}
 {bone1Name=dyn_hair_uppr_06, bone2Name=dyn_hair_master_01, swing1LimitDegrees=10, swing2LimitDegrees=5, twistLimitDegrees=0}
 {bone1Name=dyn_hair_uppr_06a, bone2Name=dyn_hair_uppr_06, swing1LimitDegrees=5, swing2LimitDegrees=5, twistLimitDegrees=0}
 {bone1Name=dyn_hair_lwr_02, bone2Name=dyn_hair_master_01, swing1LimitDegrees=5, swing2LimitDegrees=5, twistLimitDegrees=0}
 {bone1Name=dyn_hair_lwr_02a, bone2Name=dyn_hair_lwr_02, swing1LimitDegrees=5, swing2LimitDegrees=5, twistLimitDegrees=0}
 {bone1Name=dyn_hair_lwr_03, bone2Name=dyn_hair_master_01, swing1LimitDegrees=5, swing2LimitDegrees=5, twistLimitDegrees=0}
 {bone1Name=dyn_hair_lwr_03a, bone2Name=dyn_hair_lwr_03, swing1LimitDegrees=5, swing2LimitDegrees=5, twistLimitDegrees=0}
 {bone1Name=dyn_hair_mid_02, bone2Name=dyn_hair_master_01, swing1LimitDegrees=10, swing2LimitDegrees=5, twistLimitDegrees=0}
 {bone1Name=dyn_hair_uppr_01, bone2Name=dyn_hair_master_01, swing1LimitDegrees=5, swing2LimitDegrees=10, twistLimitDegrees=0}
 {bone1Name=dyn_hair_uppr_01a, bone2Name=dyn_hair_uppr_01, swing1LimitDegrees=2, swing2LimitDegrees=5, twistLimitDegrees=0}
 {bone1Name=dyn_hair_uppr_02, bone2Name=dyn_hair_master_01, swing1LimitDegrees=10, swing2LimitDegrees=5, twistLimitDegrees=0}
 {bone1Name=dyn_hair_uppr_02a, bone2Name=dyn_hair_uppr_02, swing1LimitDegrees=5, swing2LimitDegrees=5, twistLimitDegrees=0}
 {bone1Name=dyn_hair_uppr_03, bone2Name=dyn_hair_master_01, swing1LimitDegrees=10, swing2LimitDegrees=5, twistLimitDegrees=0}
 {bone1Name=dyn_hair_uppr_03a, bone2Name=dyn_hair_uppr_03, swing1LimitDegrees=5, swing2LimitDegrees=5, twistLimitDegrees=0}
 {bone1Name=dyn_hair_uppr_04, bone2Name=dyn_hair_master_01, swing1LimitDegrees=10, swing2LimitDegrees=5, twistLimitDegrees=0}
 {bone1Name=dyn_hair_uppr_07, bone2Name=dyn_hair_master_01, swing1LimitDegrees=5, swing2LimitDegrees=10, twistLimitDegrees=0}
 {bone1Name=dyn_hair_mid_03, bone2Name=dyn_hair_master_01, swing1LimitDegrees=10, swing2LimitDegrees=5, twistLimitDegrees=0}
 {bone1Name=dyn_hair_mid_03a, bone2Name=dyn_hair_mid_03, swing1LimitDegrees=5, swing2LimitDegrees=5, twistLimitDegrees=0}
 {bone1Name=dyn_hair_mid_05, bone2Name=dyn_hair_master_01, swing1LimitDegrees=5, swing2LimitDegrees=5, twistLimitDegrees=0}
 {bone1Name=dyn_hair_mid_05a, bone2Name=dyn_hair_mid_05, swing1LimitDegrees=5, swing2LimitDegrees=8, twistLimitDegrees=0}
 {bone1Name=dyn_hair_mid_01, bone2Name=dyn_hair_master_01, swing1LimitDegrees=5, swing2LimitDegrees=5, twistLimitDegrees=0}
 {bone1Name=dyn_hair_mid_01a, bone2Name=dyn_hair_mid_01, swing1LimitDegrees=5, swing2LimitDegrees=8, twistLimitDegrees=0}
 {bone1Name=dyn_hair_mid_04, bone2Name=dyn_hair_master_01, swing1LimitDegrees=8, swing2LimitDegrees=8, twistLimitDegrees=0}
 {bone1Name=dyn_hair_uppr_05, bone2Name=dyn_hair_master_01, swing1LimitDegrees=5, swing2LimitDegrees=8, twistLimitDegrees=0}
 {
  bone1Name=clavicle_l
  bone2Name=spine_05
  swing1LimitDegrees=45
  swing1Motion=Free
  swing2LimitDegrees=45
  swing2Motion=Free
  twistLimitDegrees=45
  twistMotion=Free
 }
 {
  bone1Name=clavicle_r
  bone2Name=spine_05
  swing1LimitDegrees=45
  swing1Motion=Free
  swing2LimitDegrees=45
  swing2Motion=Free
  twistLimitDegrees=45
  twistMotion=Free
 }
 {
  bone1Name=head
  bone2Name=neck_02
  swing1LimitDegrees=45
  swing1Motion=Free
  swing2LimitDegrees=45
  swing2Motion=Free
  twistLimitDegrees=45
  twistMotion=Free
 }
 {
  bone1Name=dyn_headband_master_01
  bone2Name=head
  swing1LimitDegrees=45
  swing1Motion=Free
  swing2LimitDegrees=45
  swing2Motion=Free
  twistLimitDegrees=45
  twistMotion=Free
 }
 {bone1Name=dyn_headband_l_01, bone2Name=dyn_headband_master_01, swing1LimitDegrees=10, swing2LimitDegrees=5, twistLimitDegrees=1}
 {bone1Name=dyn_headband_l_02, bone2Name=dyn_headband_l_01, swing1LimitDegrees=5, swing2LimitDegrees=5, twistLimitDegrees=1}
 {bone1Name=dyn_headband_r_01, bone2Name=dyn_headband_master_01, swing1LimitDegrees=10, swing2LimitDegrees=5, twistLimitDegrees=1}
 {bone1Name=dyn_headband_r_02, bone2Name=dyn_headband_r_01, swing1LimitDegrees=5, swing2LimitDegrees=5, twistLimitDegrees=1}
 {
  bone1Name=neck_01
  bone2Name=spine_05
  swing1LimitDegrees=45
  swing1Motion=Free
  swing2LimitDegrees=45
  swing2Motion=Free
  twistLimitDegrees=45
  twistMotion=Free
 }
 {bone1Name=dyn_eyering_l_01, bone2Name=head, swing1LimitDegrees=10, swing2LimitDegrees=10, twistLimitDegrees=1}
 {bone1Name=dyn_eyering_l_02, bone2Name=dyn_eyering_l_01, swing1LimitDegrees=15, swing2LimitDegrees=15, twistLimitDegrees=1}
 {
  bone1Name=neck_02
  bone2Name=neck_01
  swing1LimitDegrees=45
  swing1Motion=Free
  swing2LimitDegrees=45
  swing2Motion=Free
  twistLimitDegrees=45
  twistMotion=Free
 }
 {bone1Name=dyn_eyering_r_01, bone2Name=head, swing1LimitDegrees=10, swing2LimitDegrees=10, twistLimitDegrees=1}
 {bone1Name=dyn_eyering_r_02, bone2Name=dyn_eyering_r_01, swing1LimitDegrees=15, swing2LimitDegrees=15, twistLimitDegrees=1}
 {
  bone1Name=dyn_hair_lwr_04
  bone2Name=dyn_hair_master_01
  swing1LimitDegrees=45
  swing1Motion=Free
  swing2LimitDegrees=45
  swing2Motion=Free
  twistLimitDegrees=45
  twistMotion=Free
 }
]
