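# Post-process style AnimBlueprint for the FreeCanyon female face accessory, driving secondary motion of hair and headband.
# A pose copied from the parent mesh passes through an adjustment Control Rig, additive bone offsets, head- and spine-based RigidBody simulations, then a backpack Control Rig.
# The EventGraph derives corrective alphas from head pitch, crouch, movement and targeting, plus backbling state on customization load.
# Eye-ring AnimDynamics, hair constraints and scale chains are unconnected.
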AnimBlueprint /BRCosmetics/Characters/Player/Female/Medium/Bodies/F_MED_FreeCanyon/Meshes/Parts/F_MED_FreeCanyon_FaceAcc_AnimBP
target_skeleton=&/F_MED_FreeCanyon_FaceAcc_Skeleton
graphs [
 AnimGraph {nodes [
   n1 = Anim Dynamics - Bone: dyn_eyering_l_01!(ComponentPose, Alpha=1, GravityOverride:Vector="0,0,-980", ExternalForce="0,0,0") -> (Pose) {
    angularDampingOverride=1
    angularSpringConstant=5000
    bAngularSpring=true
    bOverrideAngularDamping=true
    bUseGravityOverride=true
    bUseSphericalLimits=true
    gravityScale=0.35
    simulationSpace=RootRelative
    alphaBoolBlend {blendOption=Linear}
    boundBone {boneName=dyn_eyering_l_01}
    componentAppliedLinearAccClamp=<0 0 0>
    gravityOverride {z=-980}
    physicsBodyDefinitions [{
      boundBone {boneName=dyn_eyering_l_01}
      boxExtents=<20 30 30>
      constraintSetup {angularLimitsMax=<0 5 10>, angularLimitsMin=<0 -5 -10>, angularTarget=<1 0 0>}
      localJointOffset=<4 0 0>
     }]
    sphericalLimits [{limitRadius=7.5, drivingBone {boneName=head}, sphereLocalOffset=<0 0 0>}]
   }
   n2 = Anim Dynamics - Bone: dyn_eyering_l_02!(ComponentPose=@n1, Alpha=1, GravityOverride:Vector="0,0,-980", ExternalForce="0,0,0") -> (Pose) {
    angularDampingOverride=1
    angularSpringConstant=3000
    bAngularSpring=true
    bOverrideAngularDamping=true
    bUseGravityOverride=true
    bUseSphericalLimits=true
    gravityScale=0.35
    simulationSpace=RootRelative
    alphaBoolBlend {blendOption=Linear}
    boundBone {boneName=dyn_eyering_l_02}
    componentAppliedLinearAccClamp=<0 0 0>
    gravityOverride {z=-980}
    physicsBodyDefinitions [{
      boundBone {boneName=dyn_eyering_l_02}
      boxExtents=<20 30 30>
      constraintSetup {angularLimitsMax=<0 5 15>, angularLimitsMin=<0 -5 -15>, angularTarget=<1 0 0>}
      localJointOffset=<4 0 0>
     }]
    sphericalLimits [{limitRadius=7, drivingBone {boneName=head}, sphereLocalOffset=<-6 0 0>}]
   }
   n3 = Anim Dynamics - Bone: dyn_eyering_r_01!(ComponentPose=@n2, Alpha=1, GravityOverride:Vector="0,0,-980", ExternalForce="0,0,0") -> (Pose) {
    angularDampingOverride=1
    angularSpringConstant=5000
    bAngularSpring=true
    bOverrideAngularDamping=true
    bUseGravityOverride=true
    bUseSphericalLimits=true
    gravityScale=0.35
    simulationSpace=RootRelative
    alphaBoolBlend {blendOption=Linear}
    boundBone {boneName=dyn_eyering_r_01}
    componentAppliedLinearAccClamp=<0 0 0>
    gravityOverride {z=-980}
    physicsBodyDefinitions [{
      boundBone {boneName=dyn_eyering_r_01}
      boxExtents=<20 30 30>
      constraintSetup {angularLimitsMax=<0 5 10>, angularLimitsMin=<0 -5 -10>, angularTarget=<1 0 0>}
      localJointOffset=<4 0 0>
     }]
    sphericalLimits [{limitRadius=7.5, drivingBone {boneName=head}, sphereLocalOffset=<0 0 0>}]
   }
   Anim Dynamics - Bone: dyn_eyering_r_02!(ComponentPose=@n3, Alpha=1, GravityOverride:Vector="0,0,-980", ExternalForce="0,0,0") -> (Pose) {
    angularDampingOverride=1
    angularSpringConstant=3000
    bAngularSpring=true
    bOverrideAngularDamping=true
    bUseGravityOverride=true
    bUseSphericalLimits=true
    gravityScale=0.35
    simulationSpace=RootRelative
    alphaBoolBlend {blendOption=Linear}
    boundBone {boneName=dyn_eyering_r_02}
    componentAppliedLinearAccClamp=<0 0 0>
    gravityOverride {z=-980}
    physicsBodyDefinitions [{
      boundBone {boneName=dyn_eyering_r_02}
      boxExtents=<20 30 30>
      constraintSetup {angularLimitsMax=<0 5 15>, angularLimitsMin=<0 -5 -15>, angularTarget=<1 0 0>}
      localJointOffset=<4 0 0>
     }]
    sphericalLimits [{limitRadius=7, drivingBone {boneName=head}, sphereLocalOffset=<-6 0 0>}]
   }
   n4 = "Constraint - dyn_hair_lwr_01 (PP)"!(ComponentPose, Alpha=1, ConstraintWeights_0=0.7, ConstraintWeights_1:float=0.3) -> (Pose) {
    alphaBoolBlend {blendOption=Linear}
    boneToModify {boneName=dyn_hair_lwr_01}
    constraintSetup [{targetBone {boneName=spine_05}}, {targetBone {boneName=head}}]
    constraintWeights [1,1]
   }
   n5 = "Constraint - dyn_hair_lwr_02 (PP)"!(ComponentPose=@n4, Alpha=1, ConstraintWeights_0=0.7, ConstraintWeights_1:float=0.3) -> (Pose) {
    alphaBoolBlend {blendOption=Linear}
    boneToModify {boneName=dyn_hair_lwr_02}
    constraintSetup [{targetBone {boneName=spine_05}}, {targetBone {boneName=head}}]
    constraintWeights [1,1]
   }
   n6 = "Constraint - dyn_hair_lwr_03 (PP)"!(ComponentPose=@n5, Alpha=1, ConstraintWeights_0=0.7, ConstraintWeights_1:float=0.3) -> (Pose) {
    alphaBoolBlend {blendOption=Linear}
    boneToModify {boneName=dyn_hair_lwr_03}
    constraintSetup [{targetBone {boneName=spine_05}}, {targetBone {boneName=head}}]
    constraintWeights [1,1]
   }
   "Constraint - dyn_hair_lwr_04 (PP)"!(ComponentPose=@n6, Alpha=1, ConstraintWeights_0=0.7, ConstraintWeights_1:float=0.3) -> (Pose) {
    alphaBoolBlend {blendOption=Linear}
    boneToModify {boneName=dyn_hair_lwr_04}
    constraintSetup [{targetBone {boneName=spine_05}}, {targetBone {boneName=head}}]
    constraintWeights [1,1]
   }
   "Transform (Modify) Bone - Bone: dyn_hair_master_01"(Alpha=Get Crouching_Walk() -> (Crouching_Walk:double), Translation="0,0,0", Rotation="35,0,5", Scale="1,1,1") {
    rotationMode=BMM_Additive
    rotationSpace=BCS_BoneSpace
    translationMode=BMM_Additive
    translationSpace=BCS_BoneSpace
    alphaBoolBlend {blendOption=Linear}
    alphaScaleBiasClamp {bInterpResult=true, interpSpeedDecreasing=4}
    boneToModify {boneName=dyn_hair_master_01}
    rotation {pitch=35, roll=5}
   }
   n7 = "Transform (Modify) Bone - Bone: dyn_hair_lwr_01"!(ComponentPose, bAlphaBoolEnabled:bool=True, Translation="0,0,0", Rotation="0,0,0", Scale="0.4,0.4,0.4") -> (Pose) {
    alphaInputType=Bool
    rotationMode=BMM_Additive
    rotationSpace=BCS_BoneSpace
    scaleMode=BMM_Additive
    scaleSpace=BCS_BoneSpace
    translationMode=BMM_Additive
    translationSpace=BCS_BoneSpace
    alphaBoolBlend {blendOption=Linear}
    alphaScaleBiasClamp {interpSpeedDecreasing=4}
    boneToModify {boneName=dyn_hair_lwr_01}
    scale=<0.4 0.4 0.4>
   }
   n8 = "Transform (Modify) Bone - Bone: dyn_hair_lwr_01a"!(ComponentPose=@n7, bAlphaBoolEnabled:bool=True, Translation="0,0,0", Rotation="0,0,0", Scale="0.4,0.4,0.4") -> (Pose) {
    alphaInputType=Bool
    rotationMode=BMM_Additive
    rotationSpace=BCS_BoneSpace
    scaleMode=BMM_Additive
    scaleSpace=BCS_BoneSpace
    translationMode=BMM_Additive
    translationSpace=BCS_BoneSpace
    alphaBoolBlend {blendOption=Linear}
    alphaScaleBiasClamp {interpSpeedDecreasing=4}
    boneToModify {boneName=dyn_hair_lwr_01a}
    scale=<0.4 0.4 0.4>
   }
   n9 = "Transform (Modify) Bone - Bone: dyn_hair_lwr_02"!(ComponentPose=@n8, bAlphaBoolEnabled:bool=True, Translation="0,0,0", Rotation="0,0,0", Scale="0.4,0.4,0.4") -> (Pose) {
    alphaInputType=Bool
    rotationMode=BMM_Additive
    rotationSpace=BCS_BoneSpace
    scaleMode=BMM_Additive
    scaleSpace=BCS_BoneSpace
    translationMode=BMM_Additive
    translationSpace=BCS_BoneSpace
    alphaBoolBlend {blendOption=Linear}
    alphaScaleBiasClamp {interpSpeedDecreasing=4}
    boneToModify {boneName=dyn_hair_lwr_02}
    scale=<0.4 0.4 0.4>
   }
   n10 = "Transform (Modify) Bone - Bone: dyn_hair_lwr_02a"!(ComponentPose=@n9, bAlphaBoolEnabled:bool=True, Translation="0,0,0", Rotation="0,0,0", Scale="0.4,0.4,0.4") -> (Pose) {
    alphaInputType=Bool
    rotationMode=BMM_Additive
    rotationSpace=BCS_BoneSpace
    scaleMode=BMM_Additive
    scaleSpace=BCS_BoneSpace
    translationMode=BMM_Additive
    translationSpace=BCS_BoneSpace
    alphaBoolBlend {blendOption=Linear}
    alphaScaleBiasClamp {interpSpeedDecreasing=4}
    boneToModify {boneName=dyn_hair_lwr_02a}
    scale=<0.4 0.4 0.4>
   }
   n11 = "Transform (Modify) Bone - Bone: dyn_hair_lwr_03"!(ComponentPose=@n10, bAlphaBoolEnabled:bool=True, Translation="0,0,0", Rotation="0,0,0", Scale="0.4,0.4,0.4") -> (Pose) {
    alphaInputType=Bool
    rotationMode=BMM_Additive
    rotationSpace=BCS_BoneSpace
    scaleMode=BMM_Additive
    scaleSpace=BCS_BoneSpace
    translationMode=BMM_Additive
    translationSpace=BCS_BoneSpace
    alphaBoolBlend {blendOption=Linear}
    alphaScaleBiasClamp {interpSpeedDecreasing=4}
    boneToModify {boneName=dyn_hair_lwr_03}
    scale=<0.4 0.4 0.4>
   }
   n12 = "Transform (Modify) Bone - Bone: dyn_hair_lwr_03a"!(ComponentPose=@n11, bAlphaBoolEnabled:bool=True, Translation="0,0,0", Rotation="0,0,0", Scale="0.4,0.4,0.4") -> (Pose) {
    alphaInputType=Bool
    rotationMode=BMM_Additive
    rotationSpace=BCS_BoneSpace
    scaleMode=BMM_Additive
    scaleSpace=BCS_BoneSpace
    translationMode=BMM_Additive
    translationSpace=BCS_BoneSpace
    alphaBoolBlend {blendOption=Linear}
    alphaScaleBiasClamp {interpSpeedDecreasing=4}
    boneToModify {boneName=dyn_hair_lwr_03a}
    scale=<0.4 0.4 0.4>
   }
   n13 = "Transform (Modify) Bone - Bone: dyn_hair_lwr_04"!(ComponentPose=@n12, bAlphaBoolEnabled:bool=True, Translation="0,0,0", Rotation="0,0,0", Scale="0.4,0.4,0.4") -> (Pose) {
    alphaInputType=Bool
    rotationMode=BMM_Additive
    rotationSpace=BCS_BoneSpace
    scaleMode=BMM_Additive
    scaleSpace=BCS_BoneSpace
    translationMode=BMM_Additive
    translationSpace=BCS_BoneSpace
    alphaBoolBlend {blendOption=Linear}
    alphaScaleBiasClamp {interpSpeedDecreasing=4}
    boneToModify {boneName=dyn_hair_lwr_04}
    scale=<0.4 0.4 0.4>
   }
   "Transform (Modify) Bone - Bone: dyn_hair_lwr_04a"!(ComponentPose=@n13, bAlphaBoolEnabled:bool=True, Translation="0,0,0", Rotation="0,0,0", Scale="0.4,0.4,0.4") -> (Pose) {
    alphaInputType=Bool
    rotationMode=BMM_Additive
    rotationSpace=BCS_BoneSpace
    scaleMode=BMM_Additive
    scaleSpace=BCS_BoneSpace
    translationMode=BMM_Additive
    translationSpace=BCS_BoneSpace
    alphaBoolBlend {blendOption=Linear}
    alphaScaleBiasClamp {interpSpeedDecreasing=4}
    boneToModify {boneName=dyn_hair_lwr_04a}
    scale=<0.4 0.4 0.4>
   }
   n14 = "Transform (Modify) Bone - Bone: dyn_hair_lwr_01"!(ComponentPose, Alpha:float=1, Translation="0,0,0", Rotation="0,0,0", Scale="0.2,0.2,0.2") -> (Pose) {
    rotationMode=BMM_Additive
    rotationSpace=BCS_BoneSpace
    scaleMode=BMM_Additive
    scaleSpace=BCS_BoneSpace
    translationMode=BMM_Additive
    translationSpace=BCS_BoneSpace
    alphaBoolBlend {blendOption=Linear}
    alphaScaleBiasClamp {interpSpeedDecreasing=4}
    boneToModify {boneName=dyn_hair_lwr_01}
    scale=<0.2 0.2 0.2>
   }
   n15 = "Transform (Modify) Bone - Bone: dyn_hair_lwr_01a"!(ComponentPose=@n14, Alpha:float=1, Translation="0,0,0", Rotation="0,0,0", Scale="0.2,0.2,0.2") -> (Pose) {
    rotationMode=BMM_Additive
    rotationSpace=BCS_BoneSpace
    scaleMode=BMM_Additive
    scaleSpace=BCS_BoneSpace
    translationMode=BMM_Additive
    translationSpace=BCS_BoneSpace
    alphaBoolBlend {blendOption=Linear}
    alphaScaleBiasClamp {interpSpeedDecreasing=4}
    boneToModify {boneName=dyn_hair_lwr_01a}
    scale=<0.2 0.2 0.2>
   }
   n16 = "Transform (Modify) Bone - Bone: dyn_hair_lwr_02"!(ComponentPose=@n15, Alpha:float=1, Translation="0,0,0", Rotation="0,0,0", Scale="0.2,0.2,0.2") -> (Pose) {
    rotationMode=BMM_Additive
    rotationSpace=BCS_BoneSpace
    scaleMode=BMM_Additive
    scaleSpace=BCS_BoneSpace
    translationMode=BMM_Additive
    translationSpace=BCS_BoneSpace
    alphaBoolBlend {blendOption=Linear}
    alphaScaleBiasClamp {interpSpeedDecreasing=4}
    boneToModify {boneName=dyn_hair_lwr_02}
    scale=<0.2 0.2 0.2>
   }
   n17 = "Transform (Modify) Bone - Bone: dyn_hair_lwr_02a"!(ComponentPose=@n16, Alpha:float=1, Translation="0,0,0", Rotation="0,0,0", Scale="0.2,0.2,0.2") -> (Pose) {
    rotationMode=BMM_Additive
    rotationSpace=BCS_BoneSpace
    scaleMode=BMM_Additive
    scaleSpace=BCS_BoneSpace
    translationMode=BMM_Additive
    translationSpace=BCS_BoneSpace
    alphaBoolBlend {blendOption=Linear}
    alphaScaleBiasClamp {interpSpeedDecreasing=4}
    boneToModify {boneName=dyn_hair_lwr_02a}
    scale=<0.2 0.2 0.2>
   }
   n18 = "Transform (Modify) Bone - Bone: dyn_hair_lwr_03"!(ComponentPose=@n17, Alpha:float=1, Translation="0,0,0", Rotation="0,0,0", Scale="0.2,0.2,0.2") -> (Pose) {
    rotationMode=BMM_Additive
    rotationSpace=BCS_BoneSpace
    scaleMode=BMM_Additive
    scaleSpace=BCS_BoneSpace
    translationMode=BMM_Additive
    translationSpace=BCS_BoneSpace
    alphaBoolBlend {blendOption=Linear}
    alphaScaleBiasClamp {interpSpeedDecreasing=4}
    boneToModify {boneName=dyn_hair_lwr_03}
    scale=<0.2 0.2 0.2>
   }
   n19 = "Transform (Modify) Bone - Bone: dyn_hair_lwr_03a"!(ComponentPose=@n18, Alpha:float=1, Translation="0,0,0", Rotation="0,0,0", Scale="0.2,0.2,0.2") -> (Pose) {
    rotationMode=BMM_Additive
    rotationSpace=BCS_BoneSpace
    scaleMode=BMM_Additive
    scaleSpace=BCS_BoneSpace
    translationMode=BMM_Additive
    translationSpace=BCS_BoneSpace
    alphaBoolBlend {blendOption=Linear}
    alphaScaleBiasClamp {interpSpeedDecreasing=4}
    boneToModify {boneName=dyn_hair_lwr_03a}
    scale=<0.2 0.2 0.2>
   }
   n20 = "Transform (Modify) Bone - Bone: dyn_hair_lwr_04"!(ComponentPose=@n19, Alpha:float=1, Translation="0,0,0", Rotation="0,0,0", Scale="0.2,0.2,0.2") -> (Pose) {
    rotationMode=BMM_Additive
    rotationSpace=BCS_BoneSpace
    scaleMode=BMM_Additive
    scaleSpace=BCS_BoneSpace
    translationMode=BMM_Additive
    translationSpace=BCS_BoneSpace
    alphaBoolBlend {blendOption=Linear}
    alphaScaleBiasClamp {interpSpeedDecreasing=4}
    boneToModify {boneName=dyn_hair_lwr_04}
    scale=<0.2 0.2 0.2>
   }
   "Transform (Modify) Bone - Bone: dyn_hair_lwr_04a"!(ComponentPose=@n20, Alpha:float=1, Translation="0,0,0", Rotation="0,0,0", Scale="0.2,0.2,0.2") -> (Pose) {
    rotationMode=BMM_Additive
    rotationSpace=BCS_BoneSpace
    scaleMode=BMM_Additive
    scaleSpace=BCS_BoneSpace
    translationMode=BMM_Additive
    translationSpace=BCS_BoneSpace
    alphaBoolBlend {blendOption=Linear}
    alphaScaleBiasClamp {interpSpeedDecreasing=4}
    boneToModify {boneName=dyn_hair_lwr_04a}
    scale=<0.2 0.2 0.2>
   }
   "Transform (Modify) Bone - Bone: dyn_hair_master_01"(Alpha=Get Hair_Up_Down_ADJ() -> (Hair_Up_Down_ADJ:double), Translation="0,0,0", Rotation="-20,0,0", Scale="1,1,1") {
    rotationMode=BMM_Additive
    rotationSpace=BCS_BoneSpace
    translationMode=BMM_Additive
    translationSpace=BCS_BoneSpace
    alphaBoolBlend {blendOption=Linear}
    alphaScaleBiasClamp {bInterpResult=true, interpSpeedDecreasing=2, interpSpeedIncreasing=2}
    boneToModify {boneName=dyn_hair_master_01}
    rotation {pitch=-20}
   }
   n21 = "Transform (Modify) Bone - Bone: dyn_hair_master_01"(
    ComponentPose=Local To Component(LocalPose=Use cached pose 'Initial Pose'())
    Alpha=Get HeadUp_not_targetting() -> (HeadUp_not_targetting:double)
    Translation="1,0,0"
    Rotation="40,0,0"
    Scale="1,1,1"
   ) {
    rotationMode=BMM_Additive
    rotationSpace=BCS_BoneSpace
    translationMode=BMM_Additive
    translationSpace=BCS_BoneSpace
    alphaBoolBlend {blendOption=Linear}
    alphaScaleBiasClamp {bInterpResult=true, interpSpeedDecreasing=4}
    boneToModify {boneName=dyn_hair_master_01}
    rotation {pitch=40}
    translation {x=1}
   }
   n22 = "Transform (Modify) Bone - Bone: dyn_hair_master_01"(ComponentPose=@n21, Alpha=Get Crouching() -> (Crouching:double), Translation="3,0,0", Rotation="45,0,0", Scale="1,1,1") {
    rotationMode=BMM_Additive
    rotationSpace=BCS_BoneSpace
    translationMode=BMM_Additive
    translationSpace=BCS_BoneSpace
    alphaBoolBlend {blendOption=Linear}
    alphaScaleBiasClamp {bInterpResult=true, interpSpeedDecreasing=4}
    boneToModify {boneName=dyn_hair_master_01}
    rotation {pitch=45}
    translation {x=3}
   }
   n23 = "Transform (Modify) Bone - Bone: dyn_headband_master_01"(ComponentPose=@n22, Alpha=Get Head_UP() -> (Head_UP:double), Translation="0,2,0", Rotation="30,0,0", Scale="1,1,1") {
    rotationMode=BMM_Additive
    rotationSpace=BCS_BoneSpace
    translationMode=BMM_Additive
    translationSpace=BCS_BoneSpace
    alphaBoolBlend {blendOption=Linear}
    alphaScaleBiasClamp {bInterpResult=true}
    boneToModify {boneName=dyn_headband_master_01}
    rotation {pitch=30}
    translation {y=2}
   }
   n24 = "Transform (Modify) Bone - Bone: dyn_headband_master_01"(ComponentPose=@n23, Alpha=Get Crouching(), Translation="0,0,0", Rotation="60,0,0", Scale="1,1,1") {
    rotationMode=BMM_Additive
    rotationSpace=BCS_BoneSpace
    translationMode=BMM_Additive
    translationSpace=BCS_BoneSpace
    alphaBoolBlend {blendOption=Linear}
    alphaScaleBiasClamp {bInterpResult=true}
    boneToModify {boneName=dyn_headband_master_01}
    rotation {pitch=60}
   }
   n25 = "Transform (Modify) Bone - Bone: dyn_headband_master_01"(ComponentPose=@n24, Alpha=Get Crouching_Walk(), Translation="0,0,0", Rotation="-40,0,0", Scale="1,1,1") {
    rotationMode=BMM_Additive
    rotationSpace=BCS_BoneSpace
    translationMode=BMM_Additive
    translationSpace=BCS_BoneSpace
    alphaBoolBlend {blendOption=Linear}
    alphaScaleBiasClamp {bInterpResult=true}
    boneToModify {boneName=dyn_headband_master_01}
    rotation {pitch=-40}
   }
   n26 = RigidBody!(ComponentPose=@n25, bAlphaBoolEnabled:bool=NOT Boolean(A=Get isWearingBackbling()), OverrideWorldGravity:Vector="0,0,0", ExternalForce) -> (Pose) {
    alphaInputType=Bool
    bOverrideWorldGravity=true
    overridePhysicsAsset=&/F_MED_FreeCanyon_FaceAcc_Physics
    simulationSpace=BaseBoneSpace
    alphaBoolBlend {blendOption=Linear}
    baseBoneRef {boneName=head}
    simSpaceSettings {maxAngularAcceleration=40, maxAngularVelocity=8, maxLinearAcceleration=2000, maxLinearVelocity=300, worldAlpha=1}
   }
   n27 = RigidBody(ComponentPose=@n26, OverrideWorldGravity="0,0,0") {
    alphaInputType=Bool
    bOverrideWorldGravity=true
    overridePhysicsAsset=&/F_MED_FreeCanyon_FaceAcc_SkyDiving_Physics
    simulationSpace=BaseBoneSpace
    alphaBoolBlend {blendOption=Linear}
    baseBoneRef {boneName=spine_05}
    simSpaceSettings {maxAngularAcceleration=40, maxAngularVelocity=8, maxLinearAcceleration=2000, maxLinearVelocity=300, worldAlpha=1}
   }
   n28 = "Transform (Modify) Bone - Bone: dyn_hair_master_01"(ComponentPose=@n27, Alpha=1, Translation="0,0,0", Rotation="25,0,0", Scale="1,1,1") {
    rotationMode=BMM_Additive
    rotationSpace=BCS_BoneSpace
    translationMode=BMM_Additive
    translationSpace=BCS_BoneSpace
    alphaBoolBlend {blendOption=Linear}
    alphaScaleBiasClamp {bInterpResult=true, interpSpeedDecreasing=4}
    boneToModify {boneName=dyn_hair_master_01}
    rotation {pitch=25}
   }
   n29 = "Transform (Modify) Bone - Bone: dyn_hair_master_01"(ComponentPose=@n28, Alpha=1, Translation="0,0,0", Rotation="35,0,0", Scale="1,1,1") {
    rotationMode=BMM_Additive
    rotationSpace=BCS_BoneSpace
    translationMode=BMM_Additive
    translationSpace=BCS_BoneSpace
    alphaBoolBlend {blendOption=Linear}
    alphaScaleBiasClamp {bInterpResult=true, interpSpeedDecreasing=4}
    boneToModify {boneName=dyn_hair_master_01}
    rotation {pitch=35}
   }
   n30 = "Transform (Modify) Bone - Bone: dyn_hair_uppr_04"(ComponentPose=@n29, Alpha=1, Translation="0,0,0", Rotation="0,35,0", Scale="1,1,1") {
    rotationMode=BMM_Additive
    rotationSpace=BCS_BoneSpace
    translationMode=BMM_Additive
    translationSpace=BCS_BoneSpace
    alphaBoolBlend {blendOption=Linear}
    alphaScaleBiasClamp {bInterpResult=true, interpSpeedDecreasing=4}
    boneToModify {boneName=dyn_hair_uppr_04}
    rotation {yaw=35}
   }
   n31 = "Transform (Modify) Bone - Bone: dyn_hair_uppr_05"(ComponentPose=@n30, Alpha=1, Translation="0,1.5,0", Rotation="0,15,0", Scale="1,1,1") {
    rotationMode=BMM_Additive
    rotationSpace=BCS_BoneSpace
    translationMode=BMM_Additive
    translationSpace=BCS_BoneSpace
    alphaBoolBlend {blendOption=Linear}
    alphaScaleBiasClamp {bInterpResult=true, interpSpeedDecreasing=4}
    boneToModify {boneName=dyn_hair_uppr_05}
    rotation {yaw=15}
    translation {y=1.5}
   }
   n32 = "Transform (Modify) Bone - Bone: dyn_hair_mid_03"(ComponentPose=@n31, Alpha=1, Translation="0,0,0", Rotation="0,50,0", Scale="1,1,1") {
    rotationMode=BMM_Additive
    rotationSpace=BCS_BoneSpace
    translationMode=BMM_Additive
    translationSpace=BCS_BoneSpace
    alphaBoolBlend {blendOption=Linear}
    alphaScaleBiasClamp {bInterpResult=true, interpSpeedDecreasing=4}
    boneToModify {boneName=dyn_hair_mid_03}
    rotation {yaw=50}
   }
   n33 = "Transform (Modify) Bone - Bone: dyn_hair_mid_03"(ComponentPose=@n32, Alpha=1, Translation="0,3,0", Rotation="0,15,0", Scale="1,1,1") {
    rotationMode=BMM_Additive
    rotationSpace=BCS_BoneSpace
    translationMode=BMM_Additive
    translationSpace=BCS_BoneSpace
    alphaBoolBlend {blendOption=Linear}
    alphaScaleBiasClamp {bInterpResult=true, interpSpeedDecreasing=4}
    boneToModify {boneName=dyn_hair_mid_03}
    rotation {yaw=15}
    translation {y=3}
   }
   n34 = "Transform (Modify) Bone - Bone: base_headband_master_01"(ComponentPose=@n33, Alpha=1, Translation="0,0,0", Rotation="20,0,0", Scale="1,1,1") {
    rotationMode=BMM_Additive
    rotationSpace=BCS_BoneSpace
    translationMode=BMM_Additive
    translationSpace=BCS_BoneSpace
    alphaBoolBlend {blendOption=Linear}
    alphaScaleBiasClamp {bInterpResult=true, interpSpeedDecreasing=4}
    boneToModify {boneName=base_headband_master_01}
    rotation {pitch=20}
   }
   n35 = "Transform (Modify) Bone - Bone: base_headband_master_01"(ComponentPose=@n34, Alpha=1, Translation="0,0,0", Rotation="30,0,0", Scale="1,1,1") {
    rotationMode=BMM_Additive
    rotationSpace=BCS_BoneSpace
    translationMode=BMM_Additive
    translationSpace=BCS_BoneSpace
    alphaBoolBlend {blendOption=Linear}
    alphaScaleBiasClamp {bInterpResult=true, interpSpeedDecreasing=4}
    boneToModify {boneName=base_headband_master_01}
    rotation {pitch=30}
   }
   n36 = "Transform (Modify) Bone - Bone: base_hair_master_01"(
    ComponentPose=@n35
    Alpha=Get Hair_Backpack_ADJ() -> (Hair_Backpack_ADJ:double)
    Translation="1,-0.089,5"
    Rotation="45,0.005,0.005"
    Scale="1,1,1"
   ) {
    rotationMode=BMM_Additive
    rotationSpace=BCS_BoneSpace
    scaleMode=BMM_Additive
    scaleSpace=BCS_BoneSpace
    translationMode=BMM_Additive
    translationSpace=BCS_BoneSpace
    alphaBoolBlend {blendOption=Linear}
    alphaScaleBiasClamp {bInterpResult=true, interpSpeedDecreasing=4}
    boneToModify {boneName=base_hair_master_01}
    rotation=<rot 45 0.005 0.005>
    translation=<1 -0.089 5>
   }
   n37 = Get isWearingBackbling()
   n38 = RigidBody(ComponentPose=@n36, bAlphaBoolEnabled=@n37, OverrideWorldGravity="0,0,0") {
    alphaInputType=Bool
    bOverrideWorldGravity=true
    overridePhysicsAsset=&F_MED_FreeCanyon_FaceAcc_Physics
    simulationSpace=BaseBoneSpace
    alphaBoolBlend {blendOption=Linear}
    baseBoneRef {boneName=head}
    simSpaceSettings {maxAngularAcceleration=40, maxAngularVelocity=8, maxLinearAcceleration=2000, maxLinearVelocity=300, worldAlpha=1}
   }
   n39 = Control Rig!(Source=Component To Local(ComponentPose=@n38), bAlphaBoolEnabled:bool=True, isWearingBackbling:bool=@n37) -> (Pose) {
    alphaInputType=Bool
    alphaBoolBlend {blendOption=Linear}
    controlRigAssetReference {blueprintRigClass=&/F_MED_FreeCanyon_FaceAcc_Backpack_CtrlRig.F_MED_FreeCanyon_FaceAcc_Backpack_CtrlRig_C}
   }
   Output Pose(Result=@n39) {layerGroup=None}
   n40 = Control Rig!(
    Source=Copy Pose From Mesh() {bUseAttachedParent=true}
    Alpha:float
    Crouching:bool=false
    Crouching_Walk:double=0
    Crouching_Walk_ADJ:double=Get Crouching_Walk_ADJ() -> (Crouching_Walk_ADJ:double)
    HairUp_ADJ:double=Get HairUp_ADJ() -> (HairUp_ADJ:double)
    Hair_Up_Down_ADJ:double=0
    Crouching_Walk_N:bool=false
    Crouching_Walk_N_Float:double=Get Crouching_Walk_N_Float() -> (Crouching_Walk_N_Float:double)
    isWearingBackbling_rot:bool=Get isWearingBackbling()
   ) -> (Pose) {
    alphaBoolBlend {blendOption=Linear}
    controlRigAssetReference {blueprintRigClass=&/F_MED_FreeCanyon_FaceAcc_ADJ_CtrlRig.F_MED_FreeCanyon_FaceAcc_ADJ_CtrlRig_C}
   }
   Save cached pose 'Initial Pose'(Pose=@n40) {cacheName=Initial Pose}
  ]}
 EventGraph {nodes [
   Event BlueprintUpdateAnimation()
   > n1 = Sequence()
   > n2 = "Gravity Override State Function (Parameter Asset) (Nativized)"(
    CharacterPartAnimInstance=Self-Reference()
    GameTimeInSeconds=GetGameTimeInSeconds()
    Parameters=/BRCosmetics/Characters/Player/Female/Medium/Bodies/F_MED_FreeCanyon/Meshes/M_MED_FreeCanyon_Parameters
   )
   > Set GravityOverride(GravityOverride=@n2)
   n3 = GetOwnerSkeletalMeshComponent()
   n4 = BreakRotator(InRot=BreakTransform(InTransform=GetDeltaTransformFromRefPose(self=@n3, BoneName=head, BaseName=spine_05)).Rotation)
   > Set Head_UP(Head_UP:double=MapRangeClamped(Value=@n4.Yaw, InRangeA=0, InRangeB=-75)) -> (Output_Get:double)
   > Set Head_Down(Head_Down:double=MapRangeClamped(Value=@n4.Yaw, InRangeA=0, InRangeB=75)) -> (Output_Get:double)
   n5 = BreakTransform(InTransform=GetDeltaTransformFromRefPose(self=@n3, BoneName=head, BaseName=neck_01))
   > Set Head_Down_Minor(Head_Down_Minor:double=MapRangeClamped(Value=BreakRotator(InRot=@n5.Rotation).Yaw, InRangeA=0, InRangeB=1)) -> (Output_Get:double)
   n6 = Get bIsMoving2D()
   n7 = Get bIsCrouching()
   > Set Crouching_Walk(Crouching_Walk:double=SelectFloat(A=Get Head_UP(), B=0, bPickA=AND Boolean(A=@n6, B=@n7))) -> (Output_Get:double)
   > Set Crouching(Crouching:double=SelectFloat(A=Get Head_UP(), B=0, bPickA=Get bIsCrouching())) -> (Output_Get:double)
   n8 = Get bIsCrouching()
   n9 = NOT Boolean(A=@n8)
   n10 = Get bIsTargeting()
   > Set HairUp_ADJ(HairUp_ADJ:double=SelectFloat(A=Get Head_UP(), B=0, bPickA=AND Boolean(A=@n9, B=@n10))) -> (Output_Get:double)
   n11 = AND Boolean(A=@n8, B=@n10)
   > Set Head_Down_ADJ(Head_Down_ADJ:double=SelectFloat(A=Get Head_Down() -> (Head_Down:double), B=0, bPickA=@n11)) -> (Output_Get:double)
   > Set Hair_Up_Down_ADJ(Hair_Up_Down_ADJ:double=SelectFloat(A=Get Head_Down_Minor() -> (Head_Down_Minor:double), B=0, bPickA=@n11)) -> (Output_Get:double)
   n12 = AND Boolean(A=Get bIsCrouching(), B=Get bIsTargeting())
   n13 = Get Head_UP()
   > n14 = Branch(Condition="Equal (Float)"(A=SelectFloat(A=@n13, B=0, bPickA=@n12), B=1))
   > Set Crouching_Walk_N_Float(Crouching_Walk_N_Float:double=1) -> (Output_Get:double)
   Set Crouching_Walk_N_Float(execute=@n14.else, Crouching_Walk_N_Float=0)
   Set SkyFloating_Alpha(execute=@n1.then_1, SkyFloating_Alpha:double="To Float (Boolean)"(InBool=Get bIsSkydiveFloating())) -> (Output_Get:double)
   n15 = Get bIsSkydiveDiveMode()
   n16 = Get bIsParachuteOpened()
   > Set Skydive_Alpha(Skydive_Alpha:double="To Float (Boolean)"(InBool=AND Boolean(A=@n15, B=@n16))) -> (Output_Get:double)
   n17 = Get bIsTargeting()
   n18 = Get Head_UP()
   > Set HeadUp_ADJ(HeadUp_ADJ:double=SelectFloat(A=@n18, B=0, bPickA=@n17)) -> (Output_Get:double)
   > Set HeadUp_not_targetting(HeadUp_not_targetting:double=SelectFloat(A=@n18, B=0, bPickA=NOT Boolean(A=@n17))) -> (Output_Get:double)
   n19 = NOT Boolean(A=Get bIsCrouching())
   n20 = Get bIsTargeting()
   > Set HeadUp_Standing_ADJ(HeadUp_Standing_ADJ:double=SelectFloat(A=Get Head_UP(), B=0, bPickA=AND Boolean(A=@n20, B=@n19))) -> (Output_Get:double)
   > Set Hair_Backpack_ADJ(Hair_Backpack_ADJ:double=SelectFloat(A=Get Head_UP(), B=0, bPickA=Get isWearingBackbling())) -> (Output_Get:double)
   n21 = Event OnCharacterCustomizationLoaded()
   > Set Skel_Mesh(Skel_Mesh=@n21.CustomBaseMesh)
   > n22 = Cast To PlayerPawn_Athena(Object=TryGetPawnOwner())
   > Set Body_Mesh(Body_Mesh=GetSkeletalMeshForPartType(self=@n22))
   n23 = CheckForBackbling()
   > Set isWearingBackbling(isWearingBackbling=@n23)
   > Set isWearingPackBool(isWearingPackBool:bool=OR Boolean!(A=@n23, B=@n23.bIsBackblingACape, C:bool=@n23.bIsBackblingLong) -> (ReturnValue)) -> (Output_Get:bool)
   n24 = OR Boolean!(A=@n23.bIsBackblingACape, B=@n23.bIsBackblingLong) -> (ReturnValue)
   > Set is_WearingCape(is_WearingCape:bool=@n24) -> (Output_Get:bool)
   > Set is_f_WearingCape(is_f_WearingCape:double="To Float (Boolean)"(InBool=@n24)) -> (Output_Get:double)
   > Set isWearingBP(isWearingBP:bool=AND Boolean(A=@n23, B=NOT Boolean(A=@n24))) -> (Output_Get:bool)
  ]}
]
variables [
 Body_Mesh {type=SkeletalMeshComponent}
 Crouching {type=number}
 Crouching_Walk {type=number}
 Crouching_Walk_ADJ {type=number}
 Crouching_Walk_N {type=boolean}
 Crouching_Walk_N_Float {type=number}
 GravityOverride {type=Vector}
 Hair_Backpack_ADJ {type=number}
 Hair_Up_Down_ADJ {type=number}
 HairUp_ADJ {type=number}
 Head_Down {type=number}
 Head_Down_ADJ {type=number}
 Head_Down_Minor {type=number}
 Head_UP {type=number}
 Head_UP_Bool {type=boolean}
 HeadUp_ADJ {type=number}
 HeadUp_not_targetting {type=number}
 HeadUp_Standing_ADJ {type=number}
 is_f_WearingCape {type=number}
 is_WearingCape {type=boolean}
 isWearingBackbling {type=boolean}
 isWearingBP {type=boolean}
 isWearingPackBool {type=boolean}
 Skel_Mesh {type=SkeletalMeshComponent}
 Skydive_Alpha {type=number}
 SkyFloating_Alpha {type=number}
]
